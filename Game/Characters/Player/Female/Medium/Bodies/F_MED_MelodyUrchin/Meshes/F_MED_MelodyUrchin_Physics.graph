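# Physics asset for the MelodyUrchin medium female player body, set up for secondary motion on the sweater's back and sides.
# Capsules on the pelvis, lower spine and thighs form the body volumes.
# Tiny driver bodies link with free rotation to the pelvis for center back, or to the matching thigh for the other panels.
# Each larger dyn_sweater capsule hangs off its driver through tight angular limits.
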PhysicsAsset /Game/Characters/Player/Female/Medium/Bodies/F_MED_MelodyUrchin/Meshes/F_MED_MelodyUrchin_Physics
bodies {
 driver_sweater_bck [{shapeType=Capsule, length=0.505, radius=0.505, shapeName=driver_sweater_bck_capsule, rotation=<rot 0 -90 0>}]
 driver_sweater_bck_l [{shapeType=Capsule, length=0.505, radius=0.505, shapeName=driver_sweater_bck_l_capsule, rotation=<rot 0 -90 0>}]
 driver_sweater_bck_r [{shapeType=Capsule, length=0.505, radius=0.505, shapeName=driver_sweater_bck_r_capsule, rotation=<rot 0 -90 0>}]
 driver_sweater_side_l [{shapeType=Capsule, length=0.505, radius=0.505, shapeName=driver_sweater_side_l_capsule, rotation=<rot 0 -90 0>}]
 driver_sweater_side_r [{shapeType=Capsule, length=0.505, radius=0.505, shapeName=driver_sweater_side_r_capsule, rotation=<rot 0 -90 0>}]
 dyn_sweater_bck [{shapeType=Capsule, length=12.125, radius=4.75, shapeName=dyn_sweater_bck_capsule, center=<10 0 5.5>, rotation=<rot 0 0 90>}]
 dyn_sweater_bck_l [{shapeType=Capsule, length=8.5, radius=5, shapeName=dyn_sweater_bck_l_capsule, center=<10 0 6>, rotation=<rot -0.0001 0 105>}]
 dyn_sweater_bck_r [{shapeType=Capsule, length=8.375, radius=5.125, shapeName=dyn_sweater_bck_r_capsule, center=<10 0 6>, rotation=<rot -0.0005 0 70>}]
 dyn_sweater_side_l [{shapeType=Capsule, length=17.375, radius=4, shapeName=dyn_sweater_side_l_capsule, center=<7 0 6>, rotation=<rot 0 0 90>}]
 dyn_sweater_side_r [{shapeType=Capsule, length=17, radius=4.375, shapeName=dyn_sweater_side_r_capsule, center=<7 0 6>, rotation=<rot 0 0 95>}]
 pelvis [{shapeType=Capsule, length=11.625, radius=9.5, shapeName=pelvis_capsule, center=<-7.065 -1.9907 0>}]
 spine_01 [{shapeType=Capsule, length=7.2205, radius=7.755, center=<-3.5546 -1.1825 0>}]
 thigh_l [{shapeType=Capsule, length=16.75, radius=8.25, shapeName=thigh_l_capsule, center=<-16 0 0.8875>, rotation=<rot 90 0 0>}]
 thigh_r [{shapeType=Capsule, length=16.75, radius=8.25, shapeName=thigh_r_capsule, center=<16 0 -0.8875>, rotation=<rot 90 0 0>}]
}
constraints [
 {
  bone1Name=thigh_l
  bone2Name=pelvis
  swing1LimitDegrees=45
  swing1Motion=Free
  swing2LimitDegrees=45
  swing2Motion=Free
  twistLimitDegrees=45
  twistMotion=Free
 }
 {
  bone1Name=thigh_r
  bone2Name=pelvis
  swing1LimitDegrees=45
  swing1Motion=Free
  swing2LimitDegrees=45
  swing2Motion=Free
  twistLimitDegrees=45
  twistMotion=Free
 }
 {
  bone1Name=driver_sweater_bck
  bone2Name=pelvis
  swing1LimitDegrees=45
  swing1Motion=Free
  swing2LimitDegrees=45
  swing2Motion=Free
  twistLimitDegrees=45
  twistMotion=Free
 }
 {bone1Name=dyn_sweater_bck, bone2Name=driver_sweater_bck, swing1LimitDegrees=10, swing2LimitDegrees=20, twistLimitDegrees=5}
 {bone1Name=dyn_sweater_side_r, bone2Name=driver_sweater_side_r, swing1LimitDegrees=10, swing2LimitDegrees=15, twistLimitDegrees=5}
 {
  bone1Name=driver_sweater_side_r
  bone2Name=thigh_r
  swing1LimitDegrees=45
  swing1Motion=Free
  swing2LimitDegrees=45
  swing2Motion=Free
  twistLimitDegrees=45
  twistMotion=Free
 }
 {bone1Name=dyn_sweater_bck_r, bone2Name=driver_sweater_bck_r, swing1LimitDegrees=10, swing2LimitDegrees=20, twistLimitDegrees=5}
 {
  bone1Name=driver_sweater_bck_r
  bone2Name=thigh_r
  swing1LimitDegrees=45
  swing1Motion=Free
  swing2LimitDegrees=45
  swing2Motion=Free
  twistLimitDegrees=45
  twistMotion=Free
 }
 {bone1Name=dyn_sweater_side_l, bone2Name=driver_sweater_side_l, swing1LimitDegrees=10, swing2LimitDegrees=15, twistLimitDegrees=5}
 {
  bone1Name=driver_sweater_side_l
  bone2Name=thigh_l
  swing1LimitDegrees=45
  swing1Motion=Free
  swing2LimitDegrees=45
  swing2Motion=Free
  twistLimitDegrees=45
  twistMotion=Free
 }
 {bone1Name=dyn_sweater_bck_l, bone2Name=driver_sweater_bck_l, swing1LimitDegrees=10, swing2LimitDegrees=20, twistLimitDegrees=5}
 {
  bone1Name=driver_sweater_bck_l
  bone2Name=thigh_l
  swing1LimitDegrees=45
  swing1Motion=Free
  swing2LimitDegrees=45
  swing2Motion=Free
  twistLimitDegrees=45
  twistMotion=Free
 }
]
