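# Physics asset for the legs mesh of the MelodyUrchin female medium player body, setting up physics for a dynamic "tongue" bone chain on each calf.
# A large capsule on each calf anchors a chain of small capsules running from a master bone through four tongue segments.
# Constraints link each bone to its parent with tight twist limits and swing limits that widen toward the tip, matching on left and right.
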PhysicsAsset /Game/Characters/Player/Female/Medium/Bodies/F_MED_MelodyUrchin/Meshes/F_MED_MelodyUrchin_legs_Physics
bodies {
 calf_l [{shapeType=Capsule, length=10.5, radius=8.75, shapeName=calf_l_capsule, center=<-35 0 -2>, rotation=<rot 90 0 0>}]
 calf_r [{shapeType=Capsule, length=10.5, radius=8.75, shapeName=calf_r_capsule, center=<35 0 2>, rotation=<rot 90 0 0>}]
 dyn_tongue_l_01 [{shapeType=Capsule, length=2, radius=2, shapeName=dyn_tongue_l_01_capsule, center=<1 0 0>, rotation=<rot 90 0 0>}]
 dyn_tongue_l_02 [{shapeType=Capsule, length=2, radius=2, shapeName=dyn_tongue_l_02_capsule, center=<1 0 0>, rotation=<rot 90 18.678 18.678>}]
 dyn_tongue_l_03 [{shapeType=Capsule, length=2, radius=2, shapeName=dyn_tongue_l_03_capsule, center=<1 0 0>, rotation=<rot 90 4.3173 4.3174>}]
 dyn_tongue_l_04 [{shapeType=Capsule, length=2, radius=2, shapeName=dyn_tongue_l_04_capsule, center=<1 0 0>, rotation=<rot 90 2.4667 2.4667>}]
 dyn_tongue_l_MSTR [{shapeType=Capsule, length=1, radius=1, shapeName=dyn_tongue_l_MSTR_capsule}]
 dyn_tongue_r_01 [{shapeType=Capsule, length=2, radius=2, shapeName=dyn_tongue_r_01_capsule, center=<1 0 0>, rotation=<rot 90 0 0>}]
 dyn_tongue_r_02 [{shapeType=Capsule, length=2, radius=2, shapeName=dyn_tongue_r_02_capsule, center=<1 0 0>, rotation=<rot 90 -161.322 -161.3221>}]
 dyn_tongue_r_03 [{shapeType=Capsule, length=2, radius=2, shapeName=dyn_tongue_r_03_capsule, center=<1 0 0>, rotation=<rot 90 -175.6827 -175.6827>}]
 dyn_tongue_r_04 [{shapeType=Capsule, length=2, radius=2, shapeName=dyn_tongue_r_04_capsule, center=<1 0 0>, rotation=<rot 90 -177.5333 -177.5333>}]
 dyn_tongue_r_MSTR [{shapeType=Capsule, length=1, radius=1, shapeName=dyn_tongue_r_MSTR_capsule}]
}
constraints [
 {bone1Name=dyn_tongue_l_MSTR, bone2Name=calf_l, swing1LimitDegrees=10, swing2LimitDegrees=10, twistLimitDegrees=1}
 {bone1Name=dyn_tongue_l_01, bone2Name=dyn_tongue_l_MSTR, swing1LimitDegrees=10, swing2LimitDegrees=10, twistLimitDegrees=1}
 {bone1Name=dyn_tongue_l_02, bone2Name=dyn_tongue_l_01, swing1LimitDegrees=10, swing2LimitDegrees=10, twistLimitDegrees=3}
 {bone1Name=dyn_tongue_l_03, bone2Name=dyn_tongue_l_02, swing1LimitDegrees=25, swing2LimitDegrees=25, twistLimitDegrees=3}
 {bone1Name=dyn_tongue_l_04, bone2Name=dyn_tongue_l_03, swing1LimitDegrees=25, swing2LimitDegrees=25, twistLimitDegrees=3}
 {bone1Name=dyn_tongue_r_MSTR, bone2Name=calf_r, swing1LimitDegrees=10, swing2LimitDegrees=10, twistLimitDegrees=1}
 {bone1Name=dyn_tongue_r_01, bone2Name=dyn_tongue_r_MSTR, swing1LimitDegrees=10, swing2LimitDegrees=10, twistLimitDegrees=1}
 {bone1Name=dyn_tongue_r_02, bone2Name=dyn_tongue_r_01, swing1LimitDegrees=10, swing2LimitDegrees=10, twistLimitDegrees=3}
 {bone1Name=dyn_tongue_r_03, bone2Name=dyn_tongue_r_02, swing1LimitDegrees=25, swing2LimitDegrees=25, twistLimitDegrees=3}
 {bone1Name=dyn_tongue_r_04, bone2Name=dyn_tongue_r_03, swing1LimitDegrees=25, swing2LimitDegrees=25, twistLimitDegrees=3}
]
