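# Control Rig for the MelodyUrchin player body that animates a mouth and tongue on each calf, with a separate float variable per side.
# A sequence runs tongue chain harmonics first, crossfading from a noisy wave to a fast wave as the float rises.
# It then constrains the mouth bones, deriving lip-corner offsets from upper-lip displacement and averaging intermediate lip bones.
# Finally, float-weighted transforms pose lips and tongues; the bool branches are unwired.
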ControlRigBlueprint /Game/Characters/Player/Female/Medium/Bodies/F_MED_MelodyUrchin/Meshes/F_MED_MelodyUrchin_tongues_CtrlRig
graphs [
 RigVMFunctionLibrary {links [], nodes []}
 RigVMFunction_Sequence_1_ContainedGraph {nodes [
   n1 = RigVMFunctionEntryNode()
   RigVMFunction_Sequence = &RigVMFunction_Sequence(ExecuteContext=@n1.ExecuteContext)
   n2 = &RigVMFunction_Sequence(ExecuteContext=@RigVMFunction_Sequence.B)
   RigVMFunctionReturnNode(A=@RigVMFunction_Sequence.A, B=@n2.A, C=@n2.B)
  ]}
 RigVMModel {nodes [
   RigUnit_BeginExecution = &RigUnit_BeginExecution()
   RigVMFunction_Sequence = RigVMAggregateNode(ExecuteContext=@RigUnit_BeginExecution.ExecutePin)
   n1 = &RigUnit_GetRelativeTransformForItem(Child=(Type=Bone,Name="dyn_l_mouth_upper"), Parent=(Type=Bone,Name="calf_twist_02_l"), bChildInitial=False, bParentInitial=False)
   Remap = &RigVMFunction_MathDoubleRemap(SourceMaximum=6, SourceMinimum=2, TargetMaximum=1, TargetMinimum=0, Value=@n1.RelativeTransform.Translation.X)
   &RigUnit_ModifyTransforms(
    ItemToModify="((Item=(Type=Bone,Name=\"dyn_l_mouth_lip_corner_01\"),Transform=(Rotation=(X=0,Y=0,Z=0,W=1),Translation=(X=0,Y=-1,Z=0),Scale3D=(X=1,Y=1,Z=1))),(Item=(Type=Bone,Name=\"dyn_l_mouth_lip_corner_02\"),Transform=(Rotation=(X=0,Y=0,Z=0,W=1),Translation=(X=0,Y=1,Z=0), ...(truncated)"
    WeightMinimum=0
    Weight=@Remap.Result
    ExecutePin=@RigVMFunction_Sequence.B
   )
   n2 = &RigUnit_GetRelativeTransformForItem(Child=(Type=Bone,Name="dyn_r_mouth_upper"), Parent=(Type=Bone,Name="calf_twist_02_r"), bChildInitial=False, bParentInitial=False)
   > &RigUnit_ModifyTransforms(
    ItemToModify="((Item=(Type=Bone,Name=\"dyn_r_mouth_lip_corner_01\"),Transform=(Rotation=(X=0,Y=0,Z=0,W=1),Translation=(X=0,Y=-1,Z=0),Scale3D=(X=1,Y=1,Z=1))),(Item=(Type=Bone,Name=\"dyn_r_mouth_lip_corner_02\"),Transform=(Rotation=(X=0,Y=0,Z=0,W=1),Translation=(X=0,Y=1,Z=0), ...(truncated)"
    WeightMinimum=0
    Weight=&RigVMFunction_MathDoubleRemap(SourceMaximum=-6, SourceMinimum=-2, TargetMaximum=1, TargetMinimum=0, bClamp=false, Value=@n2.RelativeTransform.Translation.X).Result
   )
   n3 = &RigVMFunction_AlphaInterp(
    InterpSpeedDecreasing=1
    InterpSpeedIncreasing=1
    Scale=1
    bClampResult=true
    bMapRange=true
    Value=RigVMVariableNode() {variable=mouth float_0}.Value
   )
   > &RigUnit_ParentConstraint(
    AdvancedSettings=(InterpolationType=Average,RotationOrderForFilter=XZY)
    Child=(Type=Bone,Name="dyn_l_mouth_upper")
    Parents=((Item=(Type=Bone,Name="calf_twist_02_l"),Weight=1))
    Weight=@n3.Result
   )
   AlphaInterp = &RigVMFunction_AlphaInterp(
    InterpSpeedDecreasing=1
    InterpSpeedIncreasing=1
    Scale=1
    bClampResult=true
    bMapRange=true
    Value=RigVMVariableNode() {variable=mouth float}.Value
   )
   > &RigUnit_ParentConstraint(
    AdvancedSettings=(InterpolationType=Average,RotationOrderForFilter=XZY)
    Child=(Type=Bone,Name="dyn_r_mouth_upper")
    Parents=((Item=(Type=Bone,Name="calf_twist_02_r"),Weight=1))
    Weight=@AlphaInterp.Result
   )
   > &RigUnit_ParentConstraint(
    AdvancedSettings=(InterpolationType=Average,RotationOrderForFilter=XZY)
    Child=(Type=Bone,Name="dyn_l_mouth_upper_01")
    Parents=((Item=(Type=Bone,Name="dyn_l_mouth_upper"),Weight=1),(Item=(Type=Bone,Name="dyn_l_mouth_lip_corner_01"),Weight=1))
    Weight=1
   )
   > &RigUnit_ParentConstraint(
    AdvancedSettings=(InterpolationType=Average,RotationOrderForFilter=XZY)
    Child=(Type=Bone,Name="dyn_l_mouth_upper_02")
    Parents=((Item=(Type=Bone,Name="dyn_l_mouth_upper"),Weight=1),(Item=(Type=Bone,Name="dyn_l_mouth_lip_corner_02"),Weight=1))
    Weight=1
   )
   > &RigUnit_ParentConstraint(
    AdvancedSettings=(InterpolationType=Average,RotationOrderForFilter=XZY)
    Child=(Type=Bone,Name="dyn_r_mouth_upper_01")
    Parents=((Item=(Type=Bone,Name="dyn_r_mouth_upper"),Weight=1),(Item=(Type=Bone,Name="dyn_r_mouth_lip_corner_01"),Weight=1))
    Weight=1
   )
   > &RigUnit_ParentConstraint(
    AdvancedSettings=(InterpolationType=Average,RotationOrderForFilter=XZY)
    Child=(Type=Bone,Name="dyn_r_mouth_upper_02")
    Parents=((Item=(Type=Bone,Name="dyn_r_mouth_upper"),Weight=1),(Item=(Type=Bone,Name="dyn_r_mouth_lip_corner_02"),Weight=1))
    Weight=1
   )
   n4 = &RigVMFunction_AlphaInterp(
    InterpSpeedDecreasing=0.5
    InterpSpeedIncreasing=2
    Scale=1
    bClampResult=true
    bMapRange=true
    Value=RigVMVariableNode() {variable=mouth float_0}.Value
   )
   &RigUnit_ModifyTransforms(
    ItemToModify="((Item=(Type=Bone,Name=\"dyn_l_mouth_upper\"),Transform=(Rotation=(X=0,Y=0.1305,Z=0,W=0.9914),Translation=(X=5,Y=0,Z=-1),Scale3D=(X=1,Y=1,Z=1))),(Item=(Type=Bone,Name=\"dyn_l_mouth_upper_01\"),Transform=(Rotation=(X=0,Y=0,Z=0,W=1),Translation=(X=2,Y=0,Z=0),Sca ...(truncated)"
    WeightMinimum=0
    Weight=@n4.Result
    ExecutePin=@RigVMFunction_Sequence.C
   )
   n5 = &RigVMFunction_AlphaInterp(
    InterpSpeedDecreasing=0.5
    InterpSpeedIncreasing=2
    Scale=1
    bClampResult=true
    bMapRange=true
    Value=RigVMVariableNode() {variable=mouth float}.Value
   )
   > &RigUnit_ModifyTransforms(
    ItemToModify="((Item=(Type=Bone,Name=\"dyn_r_mouth_upper\"),Transform=(Rotation=(X=0,Y=0.1305,Z=0,W=0.9914),Translation=(X=5,Y=0,Z=1),Scale3D=(X=1,Y=1,Z=1))),(Item=(Type=Bone,Name=\"dyn_r_mouth_upper_01\"),Transform=(Rotation=(X=0,Y=0,Z=0,W=1),Translation=(X=2,Y=0,Z=0),Scal ...(truncated)"
    WeightMinimum=0
    Weight=@n5.Result
   )
   n6 = &RigVMFunction_AlphaInterp(
    InterpSpeedDecreasing=3
    InterpSpeedIncreasing=3
    Scale=1
    bClampResult=true
    bMapRange=true
    Value=RigVMVariableNode() {variable=mouth float_0}.Value
   )
   > &RigUnit_ModifyTransforms(
    ItemToModify="((Item=(Type=Bone,Name=\"dyn_tongue_l_02\"),Transform=(Rotation=(X=0,Y=0,Z=-0.1736,W=0.9848),Translation=(X=0,Y=0,Z=0),Scale3D=(X=1,Y=1,Z=1))),(Item=(Type=Bone,Name=\"dyn_tongue_l_01\"),Transform=(Rotation=(X=0,Y=0,Z=-0.0872,W=0.9962),Translation=(X=0,Y=0,Z=0) ...(truncated)"
    WeightMinimum=0
    Weight=@n6.Result
   )
   n7 = &RigVMFunction_AlphaInterp(
    InterpSpeedDecreasing=3
    InterpSpeedIncreasing=3
    Scale=1
    bClampResult=true
    bMapRange=true
    Value=RigVMVariableNode() {variable=mouth float}.Value
   )
   > &RigUnit_ModifyTransforms(
    ItemToModify="((Item=(Type=Bone,Name=\"dyn_tongue_r_02\"),Transform=(Rotation=(X=0,Y=0,Z=-0.1736,W=0.9848),Translation=(X=0,Y=0,Z=0),Scale3D=(X=1,Y=1,Z=1))),(Item=(Type=Bone,Name=\"dyn_tongue_r_01\"),Transform=(Rotation=(X=0,Y=0,Z=-0.0872,W=0.9962),Translation=(X=0,Y=0,Z=0) ...(truncated)"
    WeightMinimum=0
    Weight=@n7.Result
   )
   n8 = &RigVMFunction_AlphaInterp(
    InterpSpeedDecreasing=3
    InterpSpeedIncreasing=5
    OutRange=(Min=1,Max=0)
    Scale=1
    bClampResult=true
    bMapRange=true
    Value=RigVMVariableNode() {variable=mouth float}.Value
   )
   /Script/ControlRig.RigUnit_ChainHarmonicsPerItem(
    ChainRoot=(Type=Bone,Name="dyn_tongue_r_01")
    DrawWorldOffset=(Rotation=(X=0,Y=0,Z=0,W=1),Translation=(X=0,Y=0,Z=0),Scale3D=(X=1,Y=1,Z=1))
    Pendulum=(bEnabled=False,PendulumStiffness=2,PendulumGravity=(X=0,Y=0,Z=0),PendulumBlend=0.75,PendulumDrag=0.98,PendulumMinimum=0,PendulumMaximum=0.1,PendulumEase=Linear,UnwindAxis=(X=0,Y=1,Z=0),UnwindMinimum=0.2,UnwindMaximum=0.05)
    Reach=(bEnabled=False,ReachTarget=(X=0,Y=0,Z=0),ReachAxis=(X=1,Y=0,Z=0),ReachMinimum=0,ReachMaximum=0,ReachEase=SineEaseInOut)
    Speed=(X=0.25,Y=0.25,Z=0.1)
    Wave=(bEnabled=true,WaveFrequency=(X=0.2,Y=0.4,Z=0.2),WaveAmplitude=(X=0.25,Y=1,Z=0.25),WaveOffset=(X=0.5,Y=0.5,Z=1),WaveNoise=(X=2.5,Y=4.5,Z=0),WaveMinimum=0,WaveMaximum=0,WaveEase=SineEaseIn)
    WaveCurve=(EditorCurveData=(Keys=((),(Time=1,Value=1)),DefaultValue=340282346638528859811704183484516925440,PreInfinityExtrap=RCCE_Constant,PostInfinityExtrap=RCCE_Constant),ExternalCurve=None)
    bDrawDebug=True
    Wave.WaveMaximum=@n8.Result
    ExecutePin=@RigVMFunction_Sequence.A
   ) {is=struct RigVMUnitNode}
   n9 = &RigVMFunction_AlphaInterp(
    InterpSpeedDecreasing=3
    InterpSpeedIncreasing=5
    OutRange=(Min=1,Max=0)
    Scale=1
    bClampResult=true
    bMapRange=true
    Value=RigVMVariableNode() {variable=mouth float_0}.Value
   )
   > &RigUnit_ChainHarmonicsPerItem(
    ChainRoot=(Type=Bone,Name="dyn_tongue_l_01")
    DrawWorldOffset=(Rotation=(X=0,Y=0,Z=0,W=1),Translation=(X=0,Y=0,Z=0),Scale3D=(X=1,Y=1,Z=1))
    Pendulum=(bEnabled=False,PendulumStiffness=2,PendulumGravity=(X=0,Y=0,Z=0),PendulumBlend=0.75,PendulumDrag=0.98,PendulumMinimum=0,PendulumMaximum=0.1,PendulumEase=Linear,UnwindAxis=(X=0,Y=1,Z=0),UnwindMinimum=0.2,UnwindMaximum=0.05)
    Reach=(bEnabled=False,ReachTarget=(X=0,Y=0,Z=0),ReachAxis=(X=1,Y=0,Z=0),ReachMinimum=0,ReachMaximum=0,ReachEase=SineEaseInOut)
    Speed=(X=0.25,Y=0.25,Z=0.1)
    Wave=(bEnabled=True,WaveFrequency=(X=0.2,Y=0.4,Z=0.2),WaveAmplitude=(X=0.25,Y=1,Z=0.25),WaveOffset=(X=0.5,Y=1.5,Z=1),WaveNoise=(X=2,Y=4.5,Z=0),WaveMinimum=0,WaveMaximum=0,WaveEase=SineEaseIn)
    WaveCurve=(EditorCurveData=(Keys=((),(Time=1,Value=1)),DefaultValue=340282346638528859811704183484516925440,PreInfinityExtrap=RCCE_Constant,PostInfinityExtrap=RCCE_Constant),ExternalCurve=None)
    bDrawDebug=True
    Wave.WaveMaximum=@n9.Result
   )
   n10 = &RigVMFunction_AlphaInterp(
    InterpSpeedDecreasing=5
    InterpSpeedIncreasing=5
    OutRange=(Min=0,Max=0.75)
    Scale=1
    bClampResult=true
    bMapRange=true
    Value=RigVMVariableNode() {variable=mouth float}.Value
   )
   > &RigUnit_ChainHarmonicsPerItem(
    ChainRoot=(Type=Bone,Name="dyn_tongue_r_01")
    DrawWorldOffset=(Rotation=(X=0,Y=0,Z=0,W=1),Translation=(X=0,Y=0,Z=0),Scale3D=(X=1,Y=1,Z=1))
    Pendulum=(bEnabled=False,PendulumStiffness=2,PendulumGravity=(X=0,Y=0,Z=0),PendulumBlend=0.75,PendulumDrag=0.98,PendulumMinimum=0,PendulumMaximum=0.1,PendulumEase=Linear,UnwindAxis=(X=0,Y=1,Z=0),UnwindMinimum=0.2,UnwindMaximum=0.05)
    Reach=(bEnabled=False,ReachTarget=(X=0,Y=0,Z=0),ReachAxis=(X=1,Y=0,Z=0),ReachMinimum=0,ReachMaximum=0,ReachEase=CircularEaseInOut)
    Speed=(X=0.25,Y=0.5,Z=-25)
    Wave=(bEnabled=True,WaveFrequency=(X=0.2,Y=0.2,Z=2),WaveAmplitude=(X=0.25,Y=0.5,Z=0.7),WaveOffset=(X=1,Y=0.25,Z=10),WaveNoise=(X=0,Y=0,Z=0),WaveMinimum=0,WaveMaximum=0.75,WaveEase=SineEaseOut)
    WaveCurve=(EditorCurveData=(Keys=((),(Time=1,Value=1)),DefaultValue=340282346638528859811704183484516925440,PreInfinityExtrap=RCCE_Constant,PostInfinityExtrap=RCCE_Constant),ExternalCurve=None)
    bDrawDebug=True
    Wave.WaveMaximum=@n10.Result
   )
   n11 = &RigVMFunction_AlphaInterp(
    InterpSpeedDecreasing=5
    InterpSpeedIncreasing=5
    OutRange=(Min=0,Max=0.75)
    Scale=1
    bClampResult=true
    bMapRange=true
    Value=RigVMVariableNode() {variable=mouth float_0}.Value
   )
   > &RigUnit_ChainHarmonicsPerItem(
    ChainRoot=(Type=Bone,Name="dyn_tongue_l_01")
    DrawWorldOffset=(Rotation=(X=0,Y=0,Z=0,W=1),Translation=(X=0,Y=0,Z=0),Scale3D=(X=1,Y=1,Z=1))
    Pendulum=(bEnabled=False,PendulumStiffness=2,PendulumGravity=(X=0,Y=0,Z=0),PendulumBlend=0.75,PendulumDrag=0.98,PendulumMinimum=0,PendulumMaximum=0.1,PendulumEase=Linear,UnwindAxis=(X=0,Y=1,Z=0),UnwindMinimum=0.2,UnwindMaximum=0.05)
    Reach=(bEnabled=False,ReachTarget=(X=0,Y=0,Z=0),ReachAxis=(X=1,Y=0,Z=0),ReachMinimum=0,ReachMaximum=0,ReachEase=SineEaseInOut)
    Speed=(X=0.25,Y=0.25,Z=-25)
    Wave=(bEnabled=True,WaveFrequency=(X=0.2,Y=0.2,Z=2),WaveAmplitude=(X=0.25,Y=0.5,Z=0.7),WaveOffset=(X=1,Y=0.25,Z=10),WaveNoise=(X=0,Y=0,Z=0),WaveMinimum=0,WaveMaximum=0.75,WaveEase=SineEaseOut)
    WaveCurve=(EditorCurveData=(Keys=((),(Time=1,Value=1)),DefaultValue=340282346638528859811704183484516925440,PreInfinityExtrap=RCCE_Constant,PostInfinityExtrap=RCCE_Constant),ExternalCurve=None)
    bDrawDebug=True
    Wave.WaveMaximum=@n11.Result
   )
   RigVMUnitNode(
    Bias=0
    ClampMax=1
    ClampMin=0
    InRange=(Min=0,Max=1)
    InterpSpeedDecreasing=10
    InterpSpeedIncreasing=10
    OutRange=(Min=0,Max=1)
    Scale=1
    Value=()
    bClampResult=False
    bInterpResult=False
    bMapRange=False
   ) {template="AlphaInterp::Execute(in Value,in Scale,in Bias,in bMapRange,in InRange,in OutRange,in bClampResult,in ClampMin,in ClampMax,in bInterpResult,in InterpSpeedIncreasing,in InterpSpeedDecreasing,out Result)"}
   RigVMFunction_ControlFlowBranch = &RigVMFunction_ControlFlowBranch(ExecuteContext=(), Condition=RigVMVariableNode() {variable=mouth bool}.Value)
   &RigVMFunction_ControlFlowBranch(Condition=False, ExecuteContext=())
  ]}
]
hierarchy [
 root {loc=<0 0 0>, rot=<0 0 0 1>}
 pelvis {parent=root, loc=<0 -0.5301 98.0571>, rot=<0.0127 0.707 -0.0127 -0.707>}
 spine_01 {parent=pelvis, loc=<6.1208 -0.9328 0>, rot=<0 0 0.034 -0.9994>}
 spine_02 {parent=spine_01, loc=<6.5996 -0.2095 0>, rot=<0 0 0.0238 -0.9997>}
 spine_03 {parent=spine_02, loc=<6.4914 -0.1237 0>, rot=<0 0 0.0149 -0.9999>}
 spine_04 {parent=spine_03, loc=<6.9598 0.0794 0>, rot=<0 0 -0.0132 -0.9999>}
 spine_05 {parent=spine_04, loc=<6.9987 0 0>, rot=<0 0 -0.0015 -1>}
 attach_cape {parent=spine_05, loc=<9.0153 -11.7137 0.0003>, rot=<0.0616 -0.7045 -0.0612 -0.7044>}
 neck_01 {parent=spine_05, loc=<11.3138 0.1529 0>, rot=<0 0 -0.2003 -0.9797>}
 neck_02 {parent=neck_01, loc=<4.5627 0 0>, rot=<0 0 0.0504 -0.9987>}
 head {parent=neck_02, loc=<5.3914 0 0>, rot=<0 0 0.115 -0.9934>}
 attach_backpack {parent=spine_05, loc=<-0.1248 -13.4728 0.0005>, rot=<-0.0092 -0.7071 0.0097 -0.707>}
 clavicle_l {parent=spine_05, loc=<5.4831 -0.1053 -3.2946>, rot=<0.0465 -0.7289 -0.009 -0.6829>}
 upperarm_l {parent=clavicle_l, loc=<9.8658 -1.2136 -0.4105>, rot=<0.0164 -0.3245 -0.0269 -0.9454>}
 lowerarm_l {parent=upperarm_l, loc=<26.465 -0.8345 0.259>, rot=<0.0036 0.0161 -0.2298 -0.9731>}
 hand_l {parent=lowerarm_l, loc=<22.7355 -0.3203 -0.6105>, rot=<-0.6784 -0.0223 0.0735 -0.7307>}
 thumb_01_l {parent=hand_l, loc=<1.3118 -0.3191 -2.459>, rot=<0.6546 -0.4042 0.0664 -0.6354>}
 thumb_02_l {parent=thumb_01_l, loc=<5.9336 -0.0448 0.0401>, rot=<-0.0037 0.0213 0.298 -0.9543>}
 thumb_03_l {parent=thumb_02_l, loc=<2.2223 0.3682 -0.0907>, rot=<0.0118 -0.013 0.0066 -0.9998>}
 index_metacarpal_l {parent=hand_l, loc=<1.6535 0.0183 -2.96>, rot=<0.0158 -0.0124 -0.0173 -0.9996>}
 index_01_l {parent=index_metacarpal_l, loc=<8.3008 0.0357 -0.0524>, rot=<-0.0044 0.0485 0.0924 -0.9945>}
 index_02_l {parent=index_01_l, loc=<3.3587 -0.3004 0.0452>, rot=<-0.002 0.0027 0.3225 -0.9466>}
 index_03_l {parent=index_02_l, loc=<2.5578 -0.123 -0.0252>, rot=<-0.002 0.0067 0.2558 -0.9667>}
 middle_metacarpal_l {parent=hand_l, loc=<2.1856 -0.1725 -0.5984>, rot=<0.006 0.0168 -0.0329 -0.9993>}
 middle_01_l {parent=middle_metacarpal_l, loc=<7.3275 -0.2447 0.0129>, rot=<-0.0208 0.0042 0.0858 -0.9961>}
 middle_02_l {parent=middle_01_l, loc=<4.0123 0.1401 0.1175>, rot=<-0.0218 0.0258 0.3512 -0.9357>}
 middle_03_l {parent=middle_02_l, loc=<2.6077 0.0534 0.1373>, rot=<0.0082 0.0188 0.1708 -0.9851>}
 ring_metacarpal_l {parent=hand_l, loc=<2.2074 -0.3625 1.3638>, rot=<-0.0284 0.0722 -0.0458 -0.9959>}
 ring_01_l {parent=ring_metacarpal_l, loc=<7.0295 -0.6197 -0.1446>, rot=<-0.0489 -0.0252 0.1892 -0.9804>}
 ring_02_l {parent=ring_01_l, loc=<3.6174 0.4647 -0.1027>, rot=<0 0 0.2806 -0.9598>}
 ring_03_l {parent=ring_02_l, loc=<2.4152 0.114 0.0199>, rot=<0 0 0.1189 -0.9929>}
 pinky_metacarpal_l {parent=hand_l, loc=<1.7616 -0.9753 3.3814>, rot=<-0.0397 0.0814 -0.069 -0.9935>}
 pinky_01_l {parent=pinky_metacarpal_l, loc=<7.1709 -0.4831 -0.1094>, rot=<-0.0654 -0.0035 0.2242 -0.9723>}
 pinky_02_l {parent=pinky_01_l, loc=<2.8274 0.0893 -0.1043>, rot=<0.0037 0.0011 0.2854 -0.9584>}
 pinky_03_l {parent=pinky_02_l, loc=<2.2046 0.0307 -0.0078>, rot=<-0.0002 0.0098 0.1445 -0.9895>}
 deform_wrist_dn_base_l {parent=hand_l, loc=<0 -1.5326 0>, rot=<0 0 0 -1>}
 deform_wrist_dn_l {parent=deform_wrist_dn_base_l, loc=<0 0 0>, rot=<0 0 0 1>}
 deform_wrist_up_base_l {parent=hand_l, loc=<0 2.3009 0>, rot=<0 0 0 -1>}
 deform_wrist_up_l {parent=deform_wrist_up_base_l, loc=<0 0 0>, rot=<0 0 0 1>}
 lowerarm_twist_02_l {parent=lowerarm_l, loc=<10.6707 0 0>, rot=<0 0 0 1>}
 lowerarm_twist_01_l {parent=lowerarm_l, loc=<18.316 0 0>, rot=<0 0 0 1>}
 deform_elbow_in_base_l {parent=lowerarm_l, loc=<0.1834 3.6062 -0.2572>, rot=<0 0 0 -1>}
 deform_elbow_in_l {parent=deform_elbow_in_base_l, loc=<0 0 0>, rot=<0 0 0 1>}
 deform_elbow_out_base_l {parent=lowerarm_l, loc=<0 -4 0>, rot=<0 0 0 1>}
 deform_elbow_out_l {parent=deform_elbow_out_base_l, loc=<0 0 0>, rot=<0 0 0 1>}
 upperarm_twist_01_l {parent=upperarm_l, loc=<7.2135 -0.1134 0>, rot=<0 0 0 1>}
 dyn_sleeve_l_MSTR {parent=upperarm_twist_01_l, loc=<0 0 0>, rot=<0 0 0 1>}
 dyn_sleeve_l_01 {parent=dyn_sleeve_l_MSTR, loc=<5.6984 0 3.4421>, rot=<0.7018 0 0 -0.7124>}
 dyn_sleeve_l_02 {parent=dyn_sleeve_l_MSTR, loc=<5.6987 4.676 0.0001>, rot=<0.964 0 0 -0.266>}
 dyn_sleeve_l_03 {parent=dyn_sleeve_l_MSTR, loc=<5.6988 -5.1857 -0.0001>, rot=<0.2516 0 0 -0.9678>}
 dyn_sleeve_l_04 {parent=dyn_sleeve_l_MSTR, loc=<5.699 0 -6.3957>, rot=<-0.7124 0 0 -0.7018>}
 dyn_sleeve_l_04_a {parent=dyn_sleeve_l_04, loc=<4.2681 0.0003 0>, rot=<0.0053 -0.005 -0.5965 -0.8025>}
 dyn_sleeve_l_04_b {parent=dyn_sleeve_l_04_a, loc=<5.6246 -0.0089 -4.8928>, rot=<-0.7025 -0.0705 -0.0705 -0.7047>}
 dyn_sleeve_l_04_c {parent=dyn_sleeve_l_04_a, loc=<5.6248 0.0103 5.443>, rot=<0.684 0.0684 -0.0725 -0.7227>}
 deform_upperarm_bk_base_l {parent=upperarm_twist_01_l, loc=<-6.4053 -4.3746 0.314>, rot=<0 0 0 -1>}
 deform_upperarm_bk_l {parent=deform_upperarm_bk_base_l, loc=<0 0 0>, rot=<0 0 0 1>}
 deform_upperarm_fr_base_l {parent=upperarm_twist_01_l, loc=<-5.6312 5.092 0.3354>, rot=<0 0 0 -1>}
 deform_upperarm_fr_l {parent=deform_upperarm_fr_base_l, loc=<0 0 0>, rot=<0 0 0 1>}
 deform_upperarm_up_base_l {parent=upperarm_twist_01_l, loc=<-8.891 0.4825 3.9517>, rot=<0 0 0 -1>}
 deform_upperarm_up_l {parent=deform_upperarm_up_base_l, loc=<0 0 0>, rot=<0 0 0 1>}
 upperarm_twist_02_l {parent=upperarm_l, loc=<19.7998 -0.1134 0>, rot=<0 0 0 1>}
 clavicle_r {parent=spine_05, loc=<5.4831 -0.1053 3.2946>, rot=<0.7289 0.0465 -0.6829 0.009>}
 upperarm_r {parent=clavicle_r, loc=<-9.8658 1.2136 0.4105>, rot=<0.0164 -0.3245 -0.0269 -0.9454>}
 lowerarm_r {parent=upperarm_r, loc=<-26.465 0.8345 -0.259>, rot=<0.0036 0.0161 -0.2298 -0.9731>}
 hand_r {parent=lowerarm_r, loc=<-22.7355 0.3203 0.6106>, rot=<-0.6784 -0.0223 0.0735 -0.7307>}
 thumb_01_r {parent=hand_r, loc=<-1.3118 0.3191 2.459>, rot=<0.6546 -0.4042 0.0664 -0.6354>}
 thumb_02_r {parent=thumb_01_r, loc=<-5.9336 0.0448 -0.0401>, rot=<-0.0037 0.0213 0.298 -0.9543>}
 thumb_03_r {parent=thumb_02_r, loc=<-2.2223 -0.3682 0.0907>, rot=<0.0118 -0.013 0.0066 -0.9998>}
 index_metacarpal_r {parent=hand_r, loc=<-1.6535 -0.0183 2.96>, rot=<0.0158 -0.0124 -0.0173 -0.9996>}
 index_01_r {parent=index_metacarpal_r, loc=<-8.3008 -0.0357 0.0524>, rot=<-0.0044 0.0485 0.0924 -0.9945>}
 index_02_r {parent=index_01_r, loc=<-3.3587 0.3004 -0.0452>, rot=<-0.002 0.0027 0.3225 -0.9466>}
 index_03_r {parent=index_02_r, loc=<-2.5578 0.123 0.0252>, rot=<-0.002 0.0067 0.2558 -0.9667>}
 middle_metacarpal_r {parent=hand_r, loc=<-2.1856 0.1725 0.5984>, rot=<0.006 0.0168 -0.0329 -0.9993>}
 middle_01_r {parent=middle_metacarpal_r, loc=<-7.3275 0.2447 -0.0129>, rot=<-0.0208 0.0042 0.0858 -0.9961>}
 middle_02_r {parent=middle_01_r, loc=<-4.0123 -0.1401 -0.1175>, rot=<-0.0218 0.0258 0.3512 -0.9357>}
 middle_03_r {parent=middle_02_r, loc=<-2.6077 -0.0534 -0.1373>, rot=<0.0082 0.0188 0.1708 -0.9851>}
 ring_metacarpal_r {parent=hand_r, loc=<-2.2074 0.3625 -1.3638>, rot=<-0.0284 0.0722 -0.0458 -0.9959>}
 ring_01_r {parent=ring_metacarpal_r, loc=<-7.0295 0.6197 0.1446>, rot=<-0.0489 -0.0252 0.1892 -0.9804>}
 ring_02_r {parent=ring_01_r, loc=<-3.6174 -0.4647 0.1027>, rot=<0 0 0.2806 -0.9598>}
 ring_03_r {parent=ring_02_r, loc=<-2.4152 -0.114 -0.0199>, rot=<0 0 0.1189 -0.9929>}
 pinky_metacarpal_r {parent=hand_r, loc=<-1.7616 0.9753 -3.3814>, rot=<-0.0397 0.0814 -0.069 -0.9935>}
 pinky_01_r {parent=pinky_metacarpal_r, loc=<-7.1709 0.4831 0.1094>, rot=<-0.0654 -0.0035 0.2242 -0.9723>}
 pinky_02_r {parent=pinky_01_r, loc=<-2.8274 -0.0893 0.1043>, rot=<0.0037 0.0011 0.2854 -0.9584>}
 pinky_03_r {parent=pinky_02_r, loc=<-2.2046 -0.0307 0.0078>, rot=<-0.0002 0.0098 0.1445 -0.9895>}
 deform_wrist_dn_base_r {parent=hand_r, loc=<-0.0001 1.5328 0>, rot=<0 0 0 -1>}
 deform_wrist_dn_r {parent=deform_wrist_dn_base_r, loc=<0 0 0>, rot=<0 0 0 1>}
 deform_wrist_up_base_r {parent=hand_r, loc=<-0.0002 -2.3007 -0.0001>, rot=<0 0 0 -1>}
 deform_wrist_up_r {parent=deform_wrist_up_base_r, loc=<0 0 0>, rot=<0 0 0 1>}
 lowerarm_twist_02_r {parent=lowerarm_r, loc=<-10.6707 0 0>, rot=<0 0 0 -1>}
 lowerarm_twist_01_r {parent=lowerarm_r, loc=<-18.316 0 0>, rot=<0 0 0 -1>}
 deform_elbow_in_base_r {parent=lowerarm_r, loc=<-0.1834 -3.6062 0.2572>, rot=<0 0 0 1>}
 deform_elbow_in_r {parent=deform_elbow_in_base_r, loc=<0 0 0>, rot=<0 0 0 1>}
 deform_elbow_out_base_r {parent=lowerarm_r, loc=<0 4 -0.0001>, rot=<0 0 0 -1>}
 deform_elbow_out_r {parent=deform_elbow_out_base_r, loc=<0 0 0>, rot=<0 0 0 1>}
 upperarm_twist_01_r {parent=upperarm_r, loc=<-7.2135 0.1134 0>, rot=<0 0 0 -1>}
 dyn_sleeve_r_MSTR {parent=upperarm_twist_01_r, loc=<0 0 0>, rot=<0 0 0 1>}
 dyn_sleeve_r_01 {parent=dyn_sleeve_r_MSTR, loc=<-5.6987 0 -3.4418>, rot=<0 0.7018 0.7124 0>}
 dyn_sleeve_r_02 {parent=dyn_sleeve_r_MSTR, loc=<-5.6987 -4.676 0>, rot=<0 -0.964 -0.266 0>}
 dyn_sleeve_r_03 {parent=dyn_sleeve_r_MSTR, loc=<-5.6987 5.1857 0>, rot=<0 0.2516 0.9678 0>}
 dyn_sleeve_r_04 {parent=dyn_sleeve_r_MSTR, loc=<-5.6987 0 6.3953>, rot=<0 -0.7124 0.7018 0>}
 dyn_sleeve_r_04_a {parent=dyn_sleeve_r_04, loc=<4.2683 0 0>, rot=<-0.0053 0.005 -0.5965 -0.8025>}
 dyn_sleeve_r_04_b {parent=dyn_sleeve_r_04_a, loc=<5.6246 -0.009 4.8928>, rot=<0.7025 0.0705 -0.0705 -0.7047>}
 dyn_sleeve_r_04_c {parent=dyn_sleeve_r_04_a, loc=<5.6246 0.01 -5.443>, rot=<-0.7047 -0.0705 -0.0705 -0.7025>}
 deform_upperarm_bk_base_r {parent=upperarm_twist_01_r, loc=<6.4053 4.3746 -0.314>, rot=<0 0 0 -1>}
 deform_upperarm_bk_r {parent=deform_upperarm_bk_base_r, loc=<0 0 0>, rot=<0 0 0 1>}
 deform_upperarm_fr_base_r {parent=upperarm_twist_01_r, loc=<5.6312 -5.092 -0.3354>, rot=<0 0 0 -1>}
 deform_upperarm_fr_r {parent=deform_upperarm_fr_base_r, loc=<0 0 0>, rot=<0 0 0 1>}
 deform_upperarm_up_base_r {parent=upperarm_twist_01_r, loc=<8.891 -0.4825 -3.9517>, rot=<0 0 0 -1>}
 deform_upperarm_up_r {parent=deform_upperarm_up_base_r, loc=<0 0 0>, rot=<0 0 0 1>}
 upperarm_twist_02_r {parent=upperarm_r, loc=<-19.7998 0.1134 0>, rot=<0 0 0 -1>}
 dyn_necklace_MSTR {parent=spine_05, loc=<7.3658 4.5679 0>, rot=<0 0 0.9277 0.3732>}
 dyn_necklace_01 {parent=dyn_necklace_MSTR, loc=<0 0 0>, rot=<0 0 0 1>}
 dyn_necklace_02 {parent=dyn_necklace_01, loc=<3.1189 0 0>, rot=<0 0 -0.1282 -0.9917>}
 dyn_necklace_03 {parent=dyn_necklace_02, loc=<4.59 -0.8653 0.0001>, rot=<-0.0002 0.0005 -0.1257 -0.9921>}
 dyn_necklace_tgt {parent=spine_05, loc=<1.592 9.338 0.0001>, rot=<0.0284 -0.7065 -0.0284 -0.7065>}
 deform_pec_aim_base_l {parent=spine_04, loc=<3.2456 9.3309 -1.2237>, rot=<0.0603 -0.4033 0.3318 -0.8507>}
 deform_pec_aim_l {parent=deform_pec_aim_base_l, loc=<0 0 0>, rot=<0 0 0 1>}
 deform_pec_base_l {parent=deform_pec_aim_l, loc=<9.3665 0 0>, rot=<0 0 0 1>}
 deform_pec_l {parent=deform_pec_base_l, loc=<0 0 0>, rot=<0 0 0 1>}
 deform_pec_aim_base_r {parent=spine_04, loc=<3.2452 9.3309 1.2237>, rot=<-0.4033 -0.0603 0.8507 0.3318>}
 deform_pec_aim_r {parent=deform_pec_aim_base_r, loc=<0 0 0>, rot=<0 0 0 1>}
 deform_pec_base_r {parent=deform_pec_aim_r, loc=<-9.3667 -0.0003 -0.0004>, rot=<0 0 0 1>}
 deform_pec_r {parent=deform_pec_base_r, loc=<0 0 0>, rot=<0 0 0 1>}
 thigh_l {parent=pelvis, loc=<-1.0855 -0.2109 -8.7053>, rot=<0.0009 0.0355 0.0155 -0.9993>}
 calf_l {parent=thigh_l, loc=<-42.5284 -0.0019 -0.7297>, rot=<0.0006 -0.0161 -0.0391 -0.9991>}
 foot_l {parent=calf_l, loc=<-46.7401 -2.5014 -0.592>, rot=<0.0427 -0.0238 0.0353 -0.9982>}
 ball_l {parent=foot_l, loc=<-3.5658 10.8066 -0.1544>, rot=<0.0001 -0.0001 -0.7063 -0.7079>}
 dyn_ankle_fang_l_01_MSTR {parent=foot_l, loc=<-5.8767 -1.0906 5.9243>, rot=<0.0391 -0.9981 -0.0446 -0.0146>}
 dyn_ankle_fang_l_01 {parent=dyn_ankle_fang_l_01_MSTR, loc=<0 0 0>, rot=<0 0 0 1>}
 dyn_ankle_fang_l_02_MSTR {parent=foot_l, loc=<-4.5757 -5.536 4.5938>, rot=<0.1041 -0.9308 0.1727 -0.3049>}
 dyn_ankle_fang_l_02 {parent=dyn_ankle_fang_l_02_MSTR, loc=<0 0 0>, rot=<0 0 0 1>}
 dyn_ankle_fang_l_03_MSTR {parent=foot_l, loc=<-0.7793 -8.2661 3.0519>, rot=<0.3897 -0.8647 0.209 -0.2384>}
 dyn_ankle_fang_l_03 {parent=dyn_ankle_fang_l_03_MSTR, loc=<0 0 0>, rot=<0 0 0 1>}
 dyn_ankle_fang_l_04_MSTR {parent=foot_l, loc=<2.169 -8.406 -0.8267>, rot=<0.3488 -0.936 -0.0461 -0.0063>}
 dyn_ankle_fang_l_04 {parent=dyn_ankle_fang_l_04_MSTR, loc=<0 0 0>, rot=<0 0 0 1>}
 dyn_ankle_fang_l_05_MSTR {parent=foot_l, loc=<-3.6128 -7.8646 -0.3952>, rot=<0.3071 -0.9506 -0.041 -0.0201>}
 dyn_ankle_fang_l_05 {parent=dyn_ankle_fang_l_05_MSTR, loc=<0 0 0>, rot=<0 0 0 1>}
 dyn_ankle_fang_l_06_MSTR {parent=foot_l, loc=<-0.1755 -6.9204 -4.1352>, rot=<0.3587 -0.893 -0.2145 0.1668>}
 dyn_ankle_fang_l_06 {parent=dyn_ankle_fang_l_06_MSTR, loc=<0 0 0>, rot=<0 0 0 1>}
 dyn_ankle_fang_l_07_MSTR {parent=foot_l, loc=<-4.4618 -5.1562 -4.975>, rot=<0.3201 -0.8767 -0.2786 0.2265>}
 dyn_ankle_fang_l_07 {parent=dyn_ankle_fang_l_07_MSTR, loc=<0 0 0>, rot=<0 0 0 1>}
 dyn_ankle_fang_l_08_MSTR {parent=foot_l, loc=<-5.5624 -0.6978 -6.9276>, rot=<0.1406 -0.9835 -0.0599 0.0964>}
 dyn_ankle_fang_l_08 {parent=dyn_ankle_fang_l_08_MSTR, loc=<0 0 0>, rot=<0 0 0 1>}
 calf_twist_02_l {parent=calf_l, loc=<-19.1546 -1.7943 -0.2638>, rot=<0.0237 0.0095 -0.0035 -0.9997>}
 dyn_mouth_l_MSTR {parent=calf_twist_02_l, loc=<-1.4951 -0.3582 1.5523>, rot=<-0.4669 0.4238 -0.5301 -0.5669>}
 dyn_mouth_l_01 {parent=dyn_mouth_l_MSTR, loc=<0 0 0>, rot=<0 0 0 1>}
 dyn_tongue_l_MSTR {parent=dyn_mouth_l_01, loc=<3.2376 -0.252 0.3102>, rot=<0.0091 0.0184 -0.2457 -0.9691>}
 dyn_tongue_l_01 {parent=dyn_tongue_l_MSTR, loc=<0 0 0>, rot=<0 0 0 1>}
 dyn_tongue_l_02 {parent=dyn_tongue_l_01, loc=<2.9256 0 0>, rot=<0.0315 0.0025 -0.3716 -0.9279>}
 dyn_tongue_l_03 {parent=dyn_tongue_l_02, loc=<3.6669 0 0>, rot=<-0.0017 0.0054 -0.126 -0.992>}
 dyn_tongue_l_04 {parent=dyn_tongue_l_03, loc=<4.3871 0 0>, rot=<-0.0004 -0.0157 -0.0134 -0.9998>}
 dyn_tongue_l_05 {parent=dyn_tongue_l_04, loc=<4.7274 0 0>, rot=<-0.0006 -0.0103 0.0742 -0.9972>}
 dyn_l_mouth_upper {parent=calf_twist_02_l, loc=<6.301 0.095 6.4087>, rot=<0.0299 -0.0257 0.9982 0.0445>}
 dyn_l_mouth_lip_corner_01 {parent=calf_twist_02_l, loc=<-0.7071 -5.4287 5.0257>, rot=<0.0299 -0.0257 0.9982 0.0445>}
 dyn_l_mouth_lip_corner_02 {parent=calf_twist_02_l, loc=<-0.4237 4.6864 3.2856>, rot=<0.0299 -0.0257 0.9982 0.0445>}
 dyn_l_mouth_upper_02 {parent=calf_twist_02_l, loc=<4.0002 3.5649 4.179>, rot=<0.0299 -0.0257 0.9982 0.0445>}
 dyn_l_mouth_upper_01 {parent=calf_twist_02_l, loc=<3.7553 -3.1221 5.9441>, rot=<0.0299 -0.0257 0.9982 0.0445>}
 calf_twist_01_l {parent=calf_l, loc=<-35.8483 -1.9197 -0.5794>, rot=<0.0237 0.0095 -0.0035 -0.9997>}
 deform_knee_in_base_l {parent=calf_l, loc=<0.8031 -5.0231 0.2456>, rot=<0 0 0 -1>}
 deform_knee_in_l {parent=deform_knee_in_base_l, loc=<0 0 0>, rot=<0 0 0 1>}
 deform_knee_out_base_l {parent=calf_l, loc=<0 4 0>, rot=<0 0 0 -1>}
 deform_knee_out_l {parent=deform_knee_out_base_l, loc=<0 0 0>, rot=<0 0 0 1>}
 thigh_twist_01_l {parent=thigh_l, loc=<-27.952 0 0>, rot=<0.0013 0 0 -1>}
 dyn_knot_l_f_MSTR {parent=thigh_l, loc=<-33.5395 6.7102 -0.1326>, rot=<0.1294 -0.0043 0.9912 -0.0269>}
 dyn_knot_l_f_01 {parent=dyn_knot_l_f_MSTR, loc=<0 0 0>, rot=<0 0 0 1>}
 dyn_knot_l_f_02 {parent=dyn_knot_l_f_01, loc=<1.8876 0 0>, rot=<-0.0007 -0.0138 0.0083 -0.9999>}
 dyn_knot_l_f_03 {parent=dyn_knot_l_f_02, loc=<1.6663 0 0>, rot=<-0.0007 -0.0181 0.0011 -0.9998>}
 dyn_knot_l_b_MSTR {parent=thigh_l, loc=<-33.5992 -6.9048 0.4778>, rot=<-0.0316 -0.9911 0.0033 0.1294>}
 dyn_knot_l_b_01 {parent=dyn_knot_l_b_MSTR, loc=<0 0 0>, rot=<0 0 0 1>}
 dyn_knot_l_b_02 {parent=dyn_knot_l_b_01, loc=<1.8876 0 0>, rot=<0.0007 0.0138 0.0083 -0.9999>}
 dyn_knot_l_b_03 {parent=dyn_knot_l_b_02, loc=<1.6663 0 0>, rot=<0.0007 0.0181 0.0011 -0.9998>}
 deform_groin_base_l {parent=thigh_l, loc=<0.4801 5.6776 -0.2486>, rot=<0 0 0 -1>}
 deform_groin_l {parent=deform_groin_base_l, loc=<0 0 0>, rot=<0 0 0 1>}
 sub_thigh_l {parent=thigh_l, loc=<-2.972 0 0>, rot=<0 0 0 -1>}
 driver_sweater_frt_l {parent=sub_thigh_l, loc=<6.7374 7.0703 -0.0995>, rot=<0 0 0 1>}
 dyn_sweater_frt_l {parent=driver_sweater_frt_l, loc=<0 0 0>, rot=<0 0.7071 0.7071 0>}
 driver_sweater_side_l {parent=sub_thigh_l, loc=<2.815 -0.4685 -5.5899>, rot=<0 0 0 1>}
 dyn_sweater_side_l {parent=driver_sweater_side_l, loc=<0 0 0>, rot=<0 -1 0 0>}
 thigh_r {parent=pelvis, loc=<-1.0853 -0.2109 8.7053>, rot=<0.0355 -0.0009 0.9993 0.0155>}
 calf_r {parent=thigh_r, loc=<42.5284 0.0019 0.7297>, rot=<0.0006 -0.0161 -0.0391 -0.9991>}
 foot_r {parent=calf_r, loc=<46.7401 2.5014 0.592>, rot=<0.0427 -0.0238 0.0353 -0.9982>}
 ball_r {parent=foot_r, loc=<3.5658 -10.8066 0.1544>, rot=<0.0001 -0.0001 -0.7063 -0.7079>}
 dyn_ankle_fang_r_01_MSTR {parent=foot_r, loc=<5.8767 1.0906 -5.9243>, rot=<0.9981 0.0391 0.0146 -0.0446>}
 dyn_ankle_fang_r_01 {parent=dyn_ankle_fang_r_01_MSTR, loc=<0 0 0>, rot=<0 0 0 1>}
 dyn_ankle_fang_r_02_MSTR {parent=foot_r, loc=<4.5757 5.536 -4.5938>, rot=<0.9308 0.1041 0.3049 0.1727>}
 dyn_ankle_fang_r_02 {parent=dyn_ankle_fang_r_02_MSTR, loc=<0 0 0>, rot=<0 0 0 1>}
 dyn_ankle_fang_r_03_MSTR {parent=foot_r, loc=<0.7793 8.2661 -3.0519>, rot=<0.8647 0.3897 0.2384 0.209>}
 dyn_ankle_fang_r_03 {parent=dyn_ankle_fang_r_03_MSTR, loc=<0 0 0>, rot=<0 0 0 1>}
 dyn_ankle_fang_r_04_MSTR {parent=foot_r, loc=<-2.169 8.406 0.8267>, rot=<0.936 0.3488 0.0063 -0.0461>}
 dyn_ankle_fang_r_04 {parent=dyn_ankle_fang_r_04_MSTR, loc=<0 0 0>, rot=<0 0 0 1>}
 dyn_ankle_fang_r_05_MSTR {parent=foot_r, loc=<3.6128 7.8646 0.3952>, rot=<0.9506 0.3071 0.0201 -0.041>}
 dyn_ankle_fang_r_05 {parent=dyn_ankle_fang_r_05_MSTR, loc=<0 0 0>, rot=<0 0 0 1>}
 dyn_ankle_fang_r_06_MSTR {parent=foot_r, loc=<0.1755 6.9205 4.1353>, rot=<0.893 0.3587 -0.1668 -0.2145>}
 dyn_ankle_fang_r_06 {parent=dyn_ankle_fang_r_06_MSTR, loc=<0 0 0>, rot=<0 0 0 1>}
 dyn_ankle_fang_r_07_MSTR {parent=foot_r, loc=<4.4618 5.1562 4.975>, rot=<0.8767 0.3201 -0.2265 -0.2786>}
 dyn_ankle_fang_r_07 {parent=dyn_ankle_fang_r_07_MSTR, loc=<0 0 0>, rot=<0 0 0 1>}
 dyn_ankle_fang_r_08_MSTR {parent=foot_r, loc=<5.5624 0.6978 6.9276>, rot=<0.9835 0.1406 -0.0964 -0.0599>}
 dyn_ankle_fang_r_08 {parent=dyn_ankle_fang_r_08_MSTR, loc=<0 0 0>, rot=<0 0 0 1>}
 calf_twist_02_r {parent=calf_r, loc=<19.1546 1.7943 0.2638>, rot=<0.0237 0.0095 -0.0035 -0.9997>}
 dyn_mouth_r_MSTR {parent=calf_twist_02_r, loc=<1.4952 0.3582 -1.5523>, rot=<-0.4238 -0.4669 0.5669 -0.5301>}
 dyn_mouth_r_01 {parent=dyn_mouth_r_MSTR, loc=<0 0 0>, rot=<0 0 0 1>}
 dyn_tongue_r_MSTR {parent=dyn_mouth_r_01, loc=<3.2376 -0.2521 -0.3102>, rot=<-0.0091 -0.0184 -0.2457 -0.9691>}
 dyn_tongue_r_01 {parent=dyn_tongue_r_MSTR, loc=<0 0 0>, rot=<0 0 0 1>}
 dyn_tongue_r_02 {parent=dyn_tongue_r_01, loc=<2.9256 0 0>, rot=<-0.0315 -0.0025 -0.3716 -0.9279>}
 dyn_tongue_r_03 {parent=dyn_tongue_r_02, loc=<3.6669 0 0>, rot=<0.0017 -0.0054 -0.126 -0.992>}
 dyn_tongue_r_04 {parent=dyn_tongue_r_03, loc=<4.3871 0 0>, rot=<0.0004 0.0157 -0.0134 -0.9998>}
 dyn_tongue_r_05 {parent=dyn_tongue_r_04, loc=<4.7274 0 0>, rot=<0.0006 0.0103 0.0742 -0.9972>}
 dyn_r_mouth_lip_corner_02 {parent=calf_twist_02_r, loc=<0.4237 -4.6864 -3.286>, rot=<-0.0257 -0.0299 0.0445 -0.9982>}
 dyn_r_mouth_upper_02 {parent=calf_twist_02_r, loc=<-4.0001 -3.5649 -4.1785>, rot=<-0.0257 -0.0299 0.0445 -0.9982>}
 dyn_r_mouth_upper {parent=calf_twist_02_r, loc=<-6.301 -0.0949 -6.4092>, rot=<-0.0257 -0.0299 0.0445 -0.9982>}
 dyn_r_mouth_upper_01 {parent=calf_twist_02_r, loc=<-3.7554 3.1221 -5.9441>, rot=<-0.0257 -0.0299 0.0445 -0.9982>}
 dyn_r_mouth_lip_corner_01 {parent=calf_twist_02_r, loc=<0.7071 5.4287 -5.0258>, rot=<-0.0257 -0.0299 0.0445 -0.9982>}
 calf_twist_01_r {parent=calf_r, loc=<35.8483 1.9197 0.5794>, rot=<0.0237 0.0095 -0.0035 -0.9997>}
 deform_knee_in_base_r {parent=calf_r, loc=<-0.8031 5.0231 -0.2457>, rot=<0 0 0 -1>}
 deform_knee_in_r {parent=deform_knee_in_base_r, loc=<0 0 0>, rot=<0 0 0 1>}
 deform_knee_out_base_r {parent=calf_r, loc=<0 -4 0>, rot=<0 0 0 -1>}
 deform_knee_out_r {parent=deform_knee_out_base_r, loc=<0 0 0>, rot=<0 0 0 1>}
 thigh_twist_01_r {parent=thigh_r, loc=<27.952 0 0>, rot=<0.0013 0 0 -1>}
 dyn_belt_r_MSTR {parent=thigh_r, loc=<33.8539 1.7562 6.5305>, rot=<0.6213 0.5076 0.4144 0.4297>}
 dyn_belt_r_01 {parent=dyn_belt_r_MSTR, loc=<0 0 0>, rot=<0 0 0 1>}
 dyn_belt_r_02 {parent=dyn_belt_r_01, loc=<2.5723 0 0>, rot=<-0.0282 0.0072 -0.2267 -0.9735>}
 dyn_belt_r_03 {parent=dyn_belt_r_02, loc=<2.5708 0 0>, rot=<-0.0478 0.0432 -0.1648 -0.9842>}
 deform_groin_base_r {parent=thigh_r, loc=<-0.48 -5.6776 0.2486>, rot=<0 0 0 -1>}
 deform_groin_r {parent=deform_groin_base_r, loc=<0 0 0>, rot=<0 0 0 1>}
 sub_thigh_r {parent=thigh_r, loc=<2.972 0 0>, rot=<0 0 0 -1>}
 driver_sweater_frt_r {parent=sub_thigh_r, loc=<-6.8151 -7.0704 0.1997>, rot=<0 0 0 1>}
 dyn_sweater_frt_r {parent=driver_sweater_frt_r, loc=<0 0 0>, rot=<-0.7071 0 0 -0.7071>}
 driver_sweater_side_r {parent=sub_thigh_r, loc=<-2.815 0.4685 5.59>, rot=<0 0 0 1>}
 dyn_sweater_side_r {parent=driver_sweater_side_r, loc=<0 0 0>, rot=<0 0 0 1>}
 dyn_shirt_MSTR {parent=pelvis, loc=<0 0 0>, rot=<0 0 0 1>}
 dyn_shirt_a_MSTR {parent=dyn_shirt_MSTR, loc=<5.1005 8.2797 5.8368>, rot=<0.0109 0.6092 0.7928 0.0142>}
 dyn_shirt_a_01 {parent=dyn_shirt_a_MSTR, loc=<0 0 0>, rot=<0 0 0 1>}
 dyn_shirt_a_02 {parent=dyn_shirt_a_01, loc=<8.8683 0.8931 1.8911>, rot=<0 0 0 1>}
 dyn_shirt_b_MSTR {parent=dyn_shirt_MSTR, loc=<4.9416 3.8481 12.1166>, rot=<0.0074 0.4132 0.9105 0.0163>}
 dyn_shirt_b_01 {parent=dyn_shirt_b_MSTR, loc=<0 0 0>, rot=<0 0 0 1>}
 dyn_shirt_b_02 {parent=dyn_shirt_b_01, loc=<8.8683 1.6922 2.2493>, rot=<0 0 0 1>}
 dyn_shirt_c_MSTR {parent=dyn_shirt_MSTR, loc=<4.5962 -5.789 12.1802>, rot=<-0.0046 -0.2588 0.9658 0.0173>}
 dyn_shirt_c_01 {parent=dyn_shirt_c_MSTR, loc=<0 0 0>, rot=<0 0 0 1>}
 dyn_shirt_c_02 {parent=dyn_shirt_c_01, loc=<8.8683 -0.1894 2.998>, rot=<0 0 0 1>}
 dyn_shirt_d_MSTR {parent=dyn_shirt_MSTR, loc=<4.4346 -10.2977 5.8373>, rot=<-0.0109 -0.6087 0.7932 0.0142>}
 dyn_shirt_d_01 {parent=dyn_shirt_d_MSTR, loc=<0 0 0>, rot=<0 0 0 1>}
 dyn_shirt_d_02 {parent=dyn_shirt_d_01, loc=<8.8683 -0.7142 2.548>, rot=<0 0 0 1>}
 dyn_shirt_e_MSTR {parent=dyn_shirt_MSTR, loc=<4.4344 -10.2977 -5.8373>, rot=<-0.0142 -0.7932 0.6087 0.0109>}
 dyn_shirt_e_01 {parent=dyn_shirt_e_MSTR, loc=<0 0 0>, rot=<0 0 0 1>}
 dyn_shirt_e_02 {parent=dyn_shirt_e_01, loc=<8.8683 0.7142 2.5479>, rot=<0 0 0 1>}
 dyn_shirt_f_MSTR {parent=dyn_shirt_MSTR, loc=<4.596 -5.789 -12.1802>, rot=<-0.0173 -0.9658 0.2588 0.0046>}
 dyn_shirt_f_01 {parent=dyn_shirt_f_MSTR, loc=<0 0 0>, rot=<0 0 0 1>}
 dyn_shirt_f_02 {parent=dyn_shirt_f_01, loc=<8.8683 0.1894 2.998>, rot=<0 0 0 1>}
 dyn_shirt_g_MSTR {parent=dyn_shirt_MSTR, loc=<4.9416 3.8481 -12.1166>, rot=<-0.0163 -0.9105 -0.4132 -0.0074>}
 dyn_shirt_g_01 {parent=dyn_shirt_g_MSTR, loc=<0 0 0>, rot=<0 0 0 1>}
 dyn_shirt_g_02 {parent=dyn_shirt_g_01, loc=<8.8683 -1.6922 2.2493>, rot=<0 0 0 1>}
 dyn_shirt_h_MSTR {parent=dyn_shirt_MSTR, loc=<5.1005 8.2796 -5.8368>, rot=<-0.0142 -0.7928 -0.6092 -0.0109>}
 dyn_shirt_h_01 {parent=dyn_shirt_h_MSTR, loc=<0 0 0>, rot=<0 0 0 1>}
 dyn_shirt_h_02 {parent=dyn_shirt_h_01, loc=<8.8683 -0.8931 1.8911>, rot=<0 0 0 1>}
 driver_sweater_bck_l {parent=pelvis, loc=<3.6926 -9.2318 -9.863>, rot=<0.0009 0.0355 0.0155 -0.9993>}
 dyn_sweater_bck_l {parent=driver_sweater_bck_l, loc=<0 0 0>, rot=<0 -0.7071 0.7071 0>}
 driver_sweater_bck_r {parent=pelvis, loc=<3.6928 -9.2319 9.863>, rot=<0.0355 -0.0009 0.9993 0.0155>}
 dyn_sweater_bck_r {parent=driver_sweater_bck_r, loc=<0 0 0>, rot=<0.7071 0 0 -0.7071>}
 deform_glute_base_l {parent=pelvis, loc=<-7.3466 -9.0213 -9.1245>, rot=<0.0172 0.031 -0.4477 -0.8935>}
 deform_glute_l {parent=deform_glute_base_l, loc=<0 0 0>, rot=<0 0 0 1>}
 deform_glute_base_r {parent=pelvis, loc=<-7.3464 -9.0213 9.1245>, rot=<0.031 -0.0172 0.8935 -0.4477>}
 deform_glute_r {parent=deform_glute_base_r, loc=<0 0 0>, rot=<0 0 0 1>}
 driver_sweater_bck {parent=pelvis, loc=<3.3846 -10.8453 0>, rot=<0 -0.7071 0 -0.7071>}
 dyn_sweater_bck {parent=driver_sweater_bck, loc=<0 0 0>, rot=<0.5 0.5 -0.5 0.5>}
 dyn_sweater_Driven_Bone {parent=pelvis, loc=<-2.6394 10.9147 0>, rot=<-0.0127 -0.707 0.0127 -0.707>}
 dyn_sweater_frt {parent=dyn_sweater_Driven_Bone, loc=<0 0 0>, rot=<0.5 -0.5 -0.5 -0.5>}
]
variables [
 mouth bool {type=bool, default=False}
 mouth bool_0 {type=bool, default=False}
 mouth float {type=double, default=0}
 mouth float_0 {type=double, default=0}
]
