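# Post-process style AnimBlueprint for the F_MED_MelodyUrchin player body, adding secondary motion to clothing and accessories.
# Copies the attached parent mesh's pose into a Control Rig driven by crouch, strafe and emote values, then runs body and legs RigidBody simulations.
# Additive bone offsets adjust sweater drivers constantly and for melee, strafing and jumping.
# Chained AnimDynamics nodes swing necklace, sleeves, knot, belt and ankle fangs, with constraint-plus-LookAt necklace aiming.
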AnimBlueprint /Game/Characters/Player/Female/Medium/Bodies/F_MED_MelodyUrchin/Meshes/F_MED_MelodyUrchin_AnimBP
target_skeleton=&/F_MED_MelodyUrchin_Skeleton
graphs [
 AnimGraph {nodes [
   RigidBody!(ComponentPose, Alpha, OverrideWorldGravity:Vector="0,0,0", ExternalForce) -> (Pose) {
    bOverrideWorldGravity=true
    overridePhysicsAsset=&/F_MED_MelodyUrchin_Physics
    simulationSpace=BaseBoneSpace
    alphaBoolBlend {blendOption=Linear}
    baseBoneRef {boneName=pelvis}
    componentAppliedLinearAccClamp=<5000 5000 5000>
    componentLinearAccScale=<0.3 0.4 0.4>
    componentLinearVelScale=<0.3 0.4 0.4>
    simSpaceSettings {
     maxAngularAcceleration=50
     maxAngularVelocity=3
     maxLinearAcceleration=350
     maxLinearVelocity=250
     velocityScaleZ=0
     worldAlpha=0.5
     externalLinearDragV=<0.2 0.2>
    }
   }
   n1 = Control Rig!(
    Source=Copy Pose From Mesh() {bUseAttachedParent=true}
    Alpha
    crouch strafe bk l:double=Get crouch strafe bk l() -> (crouch strafe bk l:float)
    crouch strafe bk r:double=Get crouch strafe bk r() -> (crouch strafe bk r:float)
    crouch:double=Get crouching() -> (crouching:double)
    emote:double=Get emote() -> (emote:double)
    strafe l:double=Get LeftMovement() -> (LeftMovement:float)
    strafe r:double=Get RightMovement() -> (RightMovement:float)
   ) -> (Pose) {
    alphaBoolBlend {blendOption=Linear}
    controlRigAssetReference {blueprintRigClass=&/F_MED_MelodyUrchin_CtrlRig.F_MED_MelodyUrchin_CtrlRig_C}
   }
   n2 = RigidBody(ComponentPose=Local To Component(LocalPose=@n1), OverrideWorldGravity="0,0,0") {
    bOverrideWorldGravity=true
    overridePhysicsAsset=&F_MED_MelodyUrchin_Physics
    simulationSpace=BaseBoneSpace
    alphaBoolBlend {blendOption=Linear}
    baseBoneRef {boneName=pelvis}
    componentAppliedLinearAccClamp=<2000 2000 2000>
    componentLinearAccScale=<0.3 0.3 0.3>
    componentLinearVelScale=<0.3 0.3 0.3>
   }
   n3 = RigidBody(ComponentPose=@n2, OverrideWorldGravity="0,0,0") {
    bOverrideWorldGravity=true
    overridePhysicsAsset=&/F_MED_MelodyUrchin_legs_Physics
    simulationSpace=BaseBoneSpace
    alphaBoolBlend {blendOption=Linear}
    baseBoneRef {boneName=root}
    componentAppliedLinearAccClamp=<2500 2500 2500>
    componentLinearAccScale=<0.3 0.3 0.3>
    componentLinearVelScale=<0.3 0.3 0.3>
   }
   n4 = Get bIsPlayingMeleeAnim()
   n5 = "Transform (Modify) Bone - Bone: driver_sweater_side_l"!(ComponentPose=@n3, bAlphaBoolEnabled:bool=@n4, Translation="0,0,0", Rotation="5,0,0", Scale="1,1,1") -> (Pose) {
    alphaInputType=Bool
    rotationMode=BMM_Additive
    rotationSpace=BCS_BoneSpace
    translationMode=BMM_Additive
    translationSpace=BCS_BoneSpace
    alphaBoolBlend {blendInTime=0.1, blendOption=Linear, blendOutTime=0.1}
    alphaScaleBiasClamp {bInterpResult=true}
    boneToModify {boneName=driver_sweater_side_l}
    rotation {pitch=5}
   }
   n6 = "Transform (Modify) Bone - Bone: driver_sweater_frt_l"!(ComponentPose=@n5, bAlphaBoolEnabled:bool=@n4, Translation="0,0,0", Rotation="3,-3,0", Scale="1,1,1") -> (Pose) {
    alphaInputType=Bool
    rotationMode=BMM_Additive
    rotationSpace=BCS_BoneSpace
    translationMode=BMM_Additive
    translationSpace=BCS_BoneSpace
    alphaBoolBlend {blendInTime=0.1, blendOption=Linear, blendOutTime=0.1}
    alphaScaleBiasClamp {bInterpResult=true}
    boneToModify {boneName=driver_sweater_frt_l}
    rotation {pitch=3, yaw=-3}
   }
   n7 = "Transform (Modify) Bone - Bone: driver_sweater_side_r"!(ComponentPose=@n6, bAlphaBoolEnabled:bool=@n4, Translation="0,0,0", Rotation="5.000,0.004,0.006", Scale="1,1,1") -> (Pose) {
    alphaInputType=Bool
    rotationMode=BMM_Additive
    rotationSpace=BCS_BoneSpace
    translationMode=BMM_Additive
    translationSpace=BCS_BoneSpace
    alphaBoolBlend {blendInTime=0.1, blendOption=Linear, blendOutTime=0.1}
    alphaScaleBiasClamp {bInterpResult=true}
    boneToModify {boneName=driver_sweater_side_r}
    rotation=<rot 5 0.004 0.006>
   }
   n8 = "Transform (Modify) Bone - Bone: driver_sweater_bck_l"!(ComponentPose=@n7, bAlphaBoolEnabled:bool=@n4, Translation="0,-0.5,0", Rotation="0,5,0", Scale="1,1,1") -> (Pose) {
    alphaInputType=Bool
    rotationMode=BMM_Additive
    rotationSpace=BCS_BoneSpace
    translationMode=BMM_Additive
    translationSpace=BCS_BoneSpace
    alphaBoolBlend {blendInTime=0.1, blendOption=Linear, blendOutTime=0.7}
    boneToModify {boneName=driver_sweater_bck_l}
    rotation {yaw=5}
    translation {y=-0.5}
   }
   n9 = "Transform (Modify) Bone - Bone: driver_sweater_bck_r"!(ComponentPose=@n8, bAlphaBoolEnabled:bool=@n4, Translation="0,0.5,0", Rotation="0.062,5,0.013", Scale="1,1,1") -> (Pose) {
    alphaInputType=Bool
    rotationMode=BMM_Additive
    rotationSpace=BCS_BoneSpace
    translationMode=BMM_Additive
    translationSpace=BCS_BoneSpace
    alphaBoolBlend {blendInTime=0.1, blendOption=Linear, blendOutTime=0.7}
    boneToModify {boneName=driver_sweater_bck_r}
    rotation=<rot 0.062 5 0.013>
    translation {y=0.5}
   }
   n10 = "Transform (Modify) Bone - Bone: driver_sweater_bck"!(ComponentPose=@n9, bAlphaBoolEnabled:bool=@n4, Translation="0,-0.5,0", Rotation="-0.052,0.11,5", Scale="1,1,1") -> (Pose) {
    alphaInputType=Bool
    rotationMode=BMM_Additive
    rotationSpace=BCS_BoneSpace
    translationMode=BMM_Additive
    translationSpace=BCS_BoneSpace
    alphaBoolBlend {blendInTime=0.1, blendOption=Linear, blendOutTime=0.7}
    boneToModify {boneName=driver_sweater_bck}
    rotation=<rot -0.052 0.11 5>
    translation {y=-0.5}
   }
   n11 = Get LeftMovement()
   n12 = "Transform (Modify) Bone - Bone: driver_sweater_side_l"!(ComponentPose=@n10, Alpha:float=@n11, Translation="0,0,0", Rotation="10,0,0", Scale="1,1,1") -> (Pose) {
    rotationMode=BMM_Additive
    rotationSpace=BCS_BoneSpace
    translationMode=BMM_Additive
    translationSpace=BCS_BoneSpace
    alphaBoolBlend {blendInTime=0.1, blendOption=Linear, blendOutTime=0.1}
    alphaScaleBiasClamp {bInterpResult=true}
    boneToModify {boneName=driver_sweater_side_l}
    rotation {pitch=10}
   }
   n13 = "Transform (Modify) Bone - Bone: driver_sweater_side_r"!(ComponentPose=@n12, Alpha:float=@n11, Translation="0,0,1", Rotation="10,0,0", Scale="1,1,1") -> (Pose) {
    rotationMode=BMM_Additive
    rotationSpace=BCS_BoneSpace
    translationMode=BMM_Additive
    translationSpace=BCS_BoneSpace
    alphaBoolBlend {blendInTime=0.1, blendOption=Linear, blendOutTime=0.1}
    alphaScaleBiasClamp {bInterpResult=true}
    boneToModify {boneName=driver_sweater_side_r}
    rotation {pitch=10}
    translation {z=1}
   }
   n14 = Get RightMovement()
   n15 = "Transform (Modify) Bone - Bone: driver_sweater_side_r"(ComponentPose=@n13, Alpha=@n14, Translation="0,0,0", Rotation="10,0,0", Scale="1,1,1") {
    rotationMode=BMM_Additive
    rotationSpace=BCS_BoneSpace
    translationMode=BMM_Additive
    translationSpace=BCS_BoneSpace
    alphaBoolBlend {blendInTime=0.1, blendOption=Linear, blendOutTime=0.1}
    alphaScaleBiasClamp {bInterpResult=true}
    boneToModify {boneName=driver_sweater_side_r}
    rotation {pitch=10}
   }
   n16 = "Transform (Modify) Bone - Bone: driver_sweater_side_l"(ComponentPose=@n15, Alpha=@n14, Translation="0,0,0", Rotation="10,0,0", Scale="1,1,1") {
    rotationMode=BMM_Additive
    rotationSpace=BCS_BoneSpace
    translationMode=BMM_Additive
    translationSpace=BCS_BoneSpace
    alphaBoolBlend {blendInTime=0.1, blendOption=Linear, blendOutTime=0.1}
    alphaScaleBiasClamp {bInterpResult=true}
    boneToModify {boneName=driver_sweater_side_l}
    rotation {pitch=10}
   }
   n17 = Get jumping() -> (jumping:double)
   n18 = "Transform (Modify) Bone - Bone: driver_sweater_bck_l"!(ComponentPose=@n16, Alpha:float=@n17, Translation="0,-5,0", Rotation="0,15,0", Scale="1,1,1") -> (Pose) {
    rotationMode=BMM_Additive
    rotationSpace=BCS_BoneSpace
    translationMode=BMM_Additive
    translationSpace=BCS_BoneSpace
    alphaBoolBlend {blendInTime=0.1, blendOption=Linear, blendOutTime=0.1}
    alphaScaleBiasClamp {bInterpResult=true, interpSpeedDecreasing=5, interpSpeedIncreasing=2}
    boneToModify {boneName=driver_sweater_bck_l}
    rotation {yaw=15}
    translation {y=-5}
   }
   n19 = "Transform (Modify) Bone - Bone: driver_sweater_bck_r"!(ComponentPose=@n18, Alpha:float=@n17, Translation="0,5,0", Rotation="0.007,15,0.003", Scale="1,1,1") -> (Pose) {
    rotationMode=BMM_Additive
    rotationSpace=BCS_BoneSpace
    translationMode=BMM_Additive
    translationSpace=BCS_BoneSpace
    alphaBoolBlend {blendInTime=0.1, blendOption=Linear, blendOutTime=0.1}
    alphaScaleBiasClamp {bInterpResult=true, interpSpeedDecreasing=5, interpSpeedIncreasing=2}
    boneToModify {boneName=driver_sweater_bck_r}
    rotation=<rot 0.007 15 0.003>
    translation {y=5}
   }
   n20 = "Transform (Modify) Bone - Bone: driver_sweater_bck"!(ComponentPose=@n19, Alpha:float=@n17, Translation="0,-5,0", Rotation="0,0,20", Scale="1,1,1") -> (Pose) {
    rotationMode=BMM_Additive
    rotationSpace=BCS_BoneSpace
    translationMode=BMM_Additive
    translationSpace=BCS_BoneSpace
    alphaBoolBlend {blendInTime=0.1, blendOption=Linear, blendOutTime=0.1}
    alphaScaleBiasClamp {bInterpResult=true, interpSpeedDecreasing=5, interpSpeedIncreasing=2}
    boneToModify {boneName=driver_sweater_bck}
    rotation {roll=20}
    translation {y=-5}
   }
   n21 = "Transform (Modify) Bone - Bone: driver_sweater_side_l"(ComponentPose=@n20, Alpha=@n17, Translation="0,0,-3", Rotation="10,0,0", Scale="1,1,1") {
    rotationMode=BMM_Additive
    rotationSpace=BCS_BoneSpace
    translationMode=BMM_Additive
    translationSpace=BCS_BoneSpace
    alphaBoolBlend {blendInTime=0.1, blendOption=Linear, blendOutTime=0.1}
    alphaScaleBiasClamp {bInterpResult=true, interpSpeedDecreasing=5, interpSpeedIncreasing=2}
    boneToModify {boneName=driver_sweater_side_l}
    rotation {pitch=10}
    translation {z=-3}
   }
   n22 = "Transform (Modify) Bone - Bone: driver_sweater_side_r"(ComponentPose=@n21, Alpha=@n17, Translation="0,0,3", Rotation="10,0,0", Scale="1,1,1") {
    rotationMode=BMM_Additive
    rotationSpace=BCS_BoneSpace
    translationMode=BMM_Additive
    translationSpace=BCS_BoneSpace
    alphaBoolBlend {blendInTime=0.1, blendOption=Linear, blendOutTime=0.1}
    alphaScaleBiasClamp {bInterpResult=true, interpSpeedDecreasing=5, interpSpeedIncreasing=2}
    boneToModify {boneName=driver_sweater_side_r}
    rotation {pitch=10}
    translation {z=3}
   }
   n23 = "Transform (Modify) Bone - Bone: driver_sweater_bck_l"!(ComponentPose=@n22, bAlphaBoolEnabled:bool=True, Translation="0,-1,0", Rotation="0,5,0", Scale="1,1,1") -> (Pose) {
    alphaInputType=Bool
    rotationMode=BMM_Additive
    rotationSpace=BCS_BoneSpace
    translationMode=BMM_Additive
    translationSpace=BCS_BoneSpace
    alphaBoolBlend {blendInTime=0.1, blendOption=Linear, blendOutTime=0.1}
    alphaScaleBiasClamp {bInterpResult=true, interpSpeedDecreasing=5, interpSpeedIncreasing=2}
    boneToModify {boneName=driver_sweater_bck_l}
    rotation {yaw=5}
    translation {y=-1}
   }
   n24 = "Transform (Modify) Bone - Bone: driver_sweater_bck_r"!(ComponentPose=@n23, bAlphaBoolEnabled:bool=True, Translation="0,1,0", Rotation="0.007,5,0.003", Scale="1,1,1") -> (Pose) {
    alphaInputType=Bool
    rotationMode=BMM_Additive
    rotationSpace=BCS_BoneSpace
    translationMode=BMM_Additive
    translationSpace=BCS_BoneSpace
    alphaBoolBlend {blendInTime=0.1, blendOption=Linear, blendOutTime=0.1}
    alphaScaleBiasClamp {bInterpResult=true, interpSpeedDecreasing=5, interpSpeedIncreasing=2}
    boneToModify {boneName=driver_sweater_bck_r}
    rotation=<rot 0.007 5 0.003>
    translation {y=1}
   }
   n25 = "Transform (Modify) Bone - Bone: driver_sweater_bck"!(ComponentPose=@n24, bAlphaBoolEnabled:bool=True, Translation="0,0,0", Rotation="0,0,2", Scale="1,1,1") -> (Pose) {
    alphaInputType=Bool
    rotationMode=BMM_Additive
    rotationSpace=BCS_BoneSpace
    translationMode=BMM_Additive
    translationSpace=BCS_BoneSpace
    alphaBoolBlend {blendInTime=0.1, blendOption=Linear, blendOutTime=0.1}
    alphaScaleBiasClamp {bInterpResult=true, interpSpeedDecreasing=5, interpSpeedIncreasing=2}
    boneToModify {boneName=driver_sweater_bck}
    rotation {roll=2}
   }
   n26 = "Transform (Modify) Bone - Bone: driver_sweater_side_l"!(ComponentPose=@n25, bAlphaBoolEnabled:bool=True, Translation="0,0,-1", Rotation="3,0,0", Scale="1,1,1") -> (Pose) {
    alphaInputType=Bool
    rotationMode=BMM_Additive
    rotationSpace=BCS_BoneSpace
    translationMode=BMM_Additive
    translationSpace=BCS_BoneSpace
    alphaBoolBlend {blendInTime=0.1, blendOption=Linear, blendOutTime=0.1}
    alphaScaleBiasClamp {bInterpResult=true, interpSpeedDecreasing=5, interpSpeedIncreasing=2}
    boneToModify {boneName=driver_sweater_side_l}
    rotation {pitch=3}
    translation {z=-1}
   }
   n27 = "Transform (Modify) Bone - Bone: driver_sweater_side_r"!(ComponentPose=@n26, bAlphaBoolEnabled:bool=True, Translation="0,0,1", Rotation="3,0,0", Scale="1,1,1") -> (Pose) {
    alphaInputType=Bool
    rotationMode=BMM_Additive
    rotationSpace=BCS_BoneSpace
    translationMode=BMM_Additive
    translationSpace=BCS_BoneSpace
    alphaBoolBlend {blendInTime=0.1, blendOption=Linear, blendOutTime=0.1}
    alphaScaleBiasClamp {bInterpResult=true, interpSpeedDecreasing=5, interpSpeedIncreasing=2}
    boneToModify {boneName=driver_sweater_side_r}
    rotation {pitch=3}
    translation {z=1}
   }
   n28 = "Constraint - dyn_necklace_tgt (PPPP)"!(
    ComponentPose=@n27
    Alpha=1
    ConstraintWeights_0=0.168
    ConstraintWeights_1:float=0.8288
    ConstraintWeights_2:float=0.0017
    ConstraintWeights_3:float=0.0016
   ) -> (Pose) {
    alphaBoolBlend {blendOption=Linear}
    boneToModify {boneName=dyn_necklace_tgt}
    constraintSetup [{targetBone {boneName=spine_04}}, {targetBone {boneName=spine_05}}, {targetBone {boneName=clavicle_l}}, {targetBone {boneName=clavicle_r}}]
    constraintWeights [1,1,1,1]
   }
   n29 = Look At - Bone: dyn_necklace_01(ComponentPose=@n28, Alpha=1) {
    alphaBoolBlend {blendOption=Linear}
    boneToModify {boneName=dyn_necklace_01}
    lookAtLocation=<0 1>
    lookAtTarget {boneReference {boneName=dyn_necklace_tgt}}
    lookAt_Axis {axis=<1 0>}
   }
   n30 = Anim Dynamics - Bone: dyn_necklace_01(ComponentPose=@n29, Alpha=1, ExternalForce="0,0,0") {
    angularDampingOverride=0.999
    angularSpringConstant=70
    bAngularSpring=true
    bGravityOverrideInSimSpace=true
    bOverrideAngularDamping=true
    bUseGravityOverride=true
    bUsePlanarLimit=false
    bUseSphericalLimits=true
    simulationSpace=RootRelative
    alphaBoolBlend {blendOption=Linear}
    boundBone {boneName=dyn_necklace_01}
    componentAppliedLinearAccClamp=<1500 1500 1500>
    componentLinearAccScale=<0.3 0.3 0.3>
    componentLinearVelScale=<0.3 0.3 0.3>
    gravityOverride {z=-980}
    physicsBodyDefinitions [{
      collisionType=InnerSphere
      sphereCollisionRadius=1
      boundBone {boneName=dyn_necklace_01}
      boxExtents=<15 20 20>
      constraintSetup {twistAxis=AxisY, angularLimitsMax=<1 2 0>, angularLimitsMin=<-1 -2 -35>, angularTarget=<1 0 0>}
      localJointOffset=<1 0 0>
     }]
    relativeSpaceBone {boneName=root}
   }
   n31 = Anim Dynamics - Bone: dyn_necklace_02(ComponentPose=@n30, Alpha=1, ExternalForce="0,0,0") {
    angularDampingOverride=0.999
    angularSpringConstant=70
    bAngularSpring=true
    bGravityOverrideInSimSpace=true
    bOverrideAngularDamping=true
    bUseGravityOverride=true
    bUsePlanarLimit=false
    bUseSphericalLimits=true
    simulationSpace=RootRelative
    alphaBoolBlend {blendOption=Linear}
    boundBone {boneName=dyn_necklace_02}
    componentAppliedLinearAccClamp=<1500 1500 1500>
    componentLinearAccScale=<0.5 0.5 0.5>
    componentLinearVelScale=<0.5 0.5 0.5>
    gravityOverride {z=-980}
    physicsBodyDefinitions [{
      collisionType=CustomSphere
      sphereCollisionRadius=1
      boundBone {boneName=dyn_necklace_02}
      boxExtents=<12 15 15>
      constraintSetup {twistAxis=AxisY, angularLimitsMax=<1 3 0>, angularLimitsMin=<-1 -3 -35>, angularTarget=<1 0 0>}
      localJointOffset=<2 0 0>
     }]
    relativeSpaceBone {boneName=root}
    sphericalLimits [
     {limitRadius=7, drivingBone {boneName=spine_04}, sphereLocalOffset=<1 4.3 4>}
     {limitRadius=7, drivingBone {boneName=spine_04}, sphereLocalOffset=<0 3.1 -4>}
    ]
   }
   n32 = Anim Dynamics - Bone: dyn_necklace_03(ComponentPose=@n31, Alpha=1, ExternalForce="0,0,0") {
    angularDampingOverride=0.999
    angularSpringConstant=70
    bAngularSpring=true
    bGravityOverrideInSimSpace=true
    bOverrideAngularDamping=true
    bUseGravityOverride=true
    bUsePlanarLimit=false
    bUseSphericalLimits=true
    simulationSpace=RootRelative
    alphaBoolBlend {blendOption=Linear}
    boundBone {boneName=dyn_necklace_03}
    componentAppliedLinearAccClamp=<1500 1500 1500>
    componentLinearAccScale=<0.5 0.5 0.5>
    componentLinearVelScale=<0.5 0.5 0.5>
    gravityOverride {z=-980}
    physicsBodyDefinitions [{
      collisionType=CustomSphere
      sphereCollisionRadius=1
      boundBone {boneName=dyn_necklace_03}
      boxExtents=<7 12 12>
      constraintSetup {twistAxis=AxisY, angularLimitsMax=<1 3 0>, angularLimitsMin=<-1 -3 -35>, angularTarget=<1 0 0>}
      localJointOffset=<2 0 0>
     }]
    relativeSpaceBone {boneName=root}
    sphericalLimits [{limitRadius=7, drivingBone {boneName=spine_04}, sphereLocalOffset=<1 4.3 4>}]
   }
   n33 = Anim Dynamics - Bone: dyn_sleeve_l_01(ComponentPose=@n32, Alpha=0.5, ExternalForce="0,0,0") {
    angularDampingOverride=0.9999
    angularSpringConstant=500
    bAngularSpring=true
    bGravityOverrideInSimSpace=true
    bOverrideAngularDamping=true
    gravityScale=0.8
    simulationSpace=BoneRelative
    alphaBoolBlend {blendOption=Linear}
    boundBone {boneName=dyn_sleeve_l_01}
    componentAppliedLinearAccClamp=<500 500 500>
    componentLinearAccScale=<1 1 1>
    componentLinearVelScale=<1 1 1>
    gravityOverride {z=-980}
    physicsBodyDefinitions [{
      sphereCollisionRadius=0
      boundBone {boneName=dyn_sleeve_l_01}
      boxExtents=<17 17 13>
      constraintSetup {angularLimitsMax=<5 15 5>, angularLimitsMin=<-5 -15 -5>, angularTarget=<1 0 0>}
      localJointOffset=<1 0 0>
     }]
    relativeSpaceBone {boneName=root}
   }
   n34 = Anim Dynamics - Bone: dyn_sleeve_l_02(ComponentPose=@n33, Alpha=0.5, ExternalForce="0,0,0") {
    angularDampingOverride=0.9999
    angularSpringConstant=500
    bAngularSpring=true
    bGravityOverrideInSimSpace=true
    bOverrideAngularDamping=true
    gravityScale=0.8
    simulationSpace=BoneRelative
    alphaBoolBlend {blendOption=Linear}
    boundBone {boneName=dyn_sleeve_l_02}
    componentAppliedLinearAccClamp=<500 500 500>
    componentLinearAccScale=<1 1 1>
    componentLinearVelScale=<1 1 1>
    gravityOverride {z=-980}
    physicsBodyDefinitions [{
      sphereCollisionRadius=0
      boundBone {boneName=dyn_sleeve_l_02}
      boxExtents=<17 17 13>
      constraintSetup {angularLimitsMax=<5 15 5>, angularLimitsMin=<-5 -15 -5>, angularTarget=<1 0 0>}
      localJointOffset=<1 0 0>
     }]
    relativeSpaceBone {boneName=root}
   }
   n35 = Anim Dynamics - Bone: dyn_sleeve_l_03(ComponentPose=@n34, Alpha=1, ExternalForce="0,0,0") {
    angularDampingOverride=0.9999
    angularSpringConstant=500
    bAngularSpring=true
    bGravityOverrideInSimSpace=true
    bOverrideAngularDamping=true
    gravityScale=0.8
    simulationSpace=BoneRelative
    alphaBoolBlend {blendOption=Linear}
    boundBone {boneName=dyn_sleeve_l_03}
    componentAppliedLinearAccClamp=<500 500 500>
    componentLinearAccScale=<1 1 1>
    componentLinearVelScale=<1 1 1>
    gravityOverride {z=-980}
    physicsBodyDefinitions [{
      sphereCollisionRadius=0
      boundBone {boneName=dyn_sleeve_l_03}
      boxExtents=<17 17 13>
      constraintSetup {angularLimitsMax=<5 15 5>, angularLimitsMin=<-5 -15 -5>, angularTarget=<1 0 0>}
      localJointOffset=<1 0 0>
     }]
    relativeSpaceBone {boneName=root}
   }
   n36 = Anim Dynamics - Bone: dyn_sleeve_l_04(ComponentPose=@n35, Alpha=0.5, ExternalForce="0,0,0") {
    angularDampingOverride=0.9999
    angularSpringConstant=3500
    bAngularSpring=true
    bGravityOverrideInSimSpace=true
    bOverrideAngularDamping=true
    gravityScale=0.8
    simulationSpace=BoneRelative
    alphaBoolBlend {blendOption=Linear}
    boundBone {boneName=dyn_sleeve_l_04}
    componentAppliedLinearAccClamp=<500 500 500>
    componentLinearAccScale=<1 1 1>
    componentLinearVelScale=<1 1 1>
    gravityOverride {z=-980}
    physicsBodyDefinitions [{
      sphereCollisionRadius=0
      boundBone {boneName=dyn_sleeve_l_04}
      boxExtents=<17 17 13>
      constraintSetup {angularLimitsMax=<1 10 15>, angularLimitsMin=<-1 1 0>, angularTarget=<1 0 0>, linearAxesMax=<1 0 0>, linearAxesMin=<-1 0 0>}
      localJointOffset=<1 0 0>
     }]
    relativeSpaceBone {boneName=root}
   }
   n37 = Anim Dynamics - Bone: dyn_sleeve_l_04_a(ComponentPose=@n36, Alpha=0.5, ExternalForce="0,0,0") {
    angularDampingOverride=0.9999
    angularSpringConstant=500
    bAngularSpring=true
    bGravityOverrideInSimSpace=true
    bOverrideAngularDamping=true
    gravityScale=0.8
    simulationSpace=BoneRelative
    alphaBoolBlend {blendOption=Linear}
    boundBone {boneName=dyn_sleeve_l_04_a}
    componentAppliedLinearAccClamp=<500 500 500>
    componentLinearAccScale=<1 1 1>
    componentLinearVelScale=<1 1 1>
    gravityOverride {z=-980}
    physicsBodyDefinitions [{
      sphereCollisionRadius=0
      boundBone {boneName=dyn_sleeve_l_04_a}
      boxExtents=<17 17 13>
      constraintSetup {angularLimitsMax=<1 2 5>, angularLimitsMin=<-1 -2 -3>, angularTarget=<1 0 0>}
      localJointOffset=<1 0 0>
     }]
    relativeSpaceBone {boneName=root}
   }
   n38 = Anim Dynamics - Bone: dyn_sleeve_l_04_b(ComponentPose=@n37, Alpha=1, ExternalForce="0,0,0") {
    angularDampingOverride=0.9999
    angularSpringConstant=500
    bAngularSpring=true
    bGravityOverrideInSimSpace=true
    bOverrideAngularDamping=true
    gravityScale=0.8
    simulationSpace=BoneRelative
    alphaBoolBlend {blendOption=Linear}
    boundBone {boneName=dyn_sleeve_l_04_b}
    componentAppliedLinearAccClamp=<500 500 500>
    componentLinearAccScale=<1 1 1>
    componentLinearVelScale=<1 1 1>
    gravityOverride {z=-980}
    physicsBodyDefinitions [{
      sphereCollisionRadius=0
      boundBone {boneName=dyn_sleeve_l_04_b}
      boxExtents=<17 17 13>
      constraintSetup {angularLimitsMax=<5 5 2>, angularLimitsMin=<-5 -5 -2>, angularTarget=<1 0 0>}
      localJointOffset=<2 0 0>
     }]
    relativeSpaceBone {boneName=root}
   }
   n39 = Anim Dynamics - Bone: dyn_sleeve_l_04_c(ComponentPose=@n38, Alpha=1, ExternalForce="0,0,0") {
    angularDampingOverride=0.9999
    angularSpringConstant=500
    bAngularSpring=true
    bGravityOverrideInSimSpace=true
    bOverrideAngularDamping=true
    gravityScale=0.8
    simulationSpace=BoneRelative
    alphaBoolBlend {blendOption=Linear}
    boundBone {boneName=dyn_sleeve_l_04_c}
    componentAppliedLinearAccClamp=<500 500 500>
    componentLinearAccScale=<1 1 1>
    componentLinearVelScale=<1 1 1>
    gravityOverride {z=-980}
    physicsBodyDefinitions [{
      sphereCollisionRadius=0
      boundBone {boneName=dyn_sleeve_l_04_c}
      boxExtents=<17 17 13>
      constraintSetup {angularLimitsMax=<5 15 5>, angularLimitsMin=<-5 -15 -5>, angularTarget=<1 0 0>}
      localJointOffset=<2 0 0>
     }]
    relativeSpaceBone {boneName=root}
   }
   n40 = Anim Dynamics - Bone: dyn_sleeve_r_01(ComponentPose=@n39, Alpha=0.5, ExternalForce="0,0,0") {
    angularDampingOverride=0.9999
    angularSpringConstant=500
    bAngularSpring=true
    bGravityOverrideInSimSpace=true
    bOverrideAngularDamping=true
    gravityScale=0.8
    simulationSpace=BoneRelative
    alphaBoolBlend {blendOption=Linear}
    boundBone {boneName=dyn_sleeve_r_01}
    componentAppliedLinearAccClamp=<500 500 500>
    componentLinearAccScale=<1 1 1>
    componentLinearVelScale=<1 1 1>
    gravityOverride {z=-980}
    physicsBodyDefinitions [{
      sphereCollisionRadius=0
      boundBone {boneName=dyn_sleeve_r_01}
      boxExtents=<17 17 13>
      constraintSetup {angularLimitsMax=<5 15 5>, angularLimitsMin=<-5 -15 -5>, angularTarget=<1 0 0>}
      localJointOffset=<1 0 0>
     }]
    relativeSpaceBone {boneName=root}
   }
   n41 = Anim Dynamics - Bone: dyn_sleeve_r_02(ComponentPose=@n40, Alpha=0.5, ExternalForce="0,0,0") {
    angularDampingOverride=0.9999
    angularSpringConstant=500
    bAngularSpring=true
    bGravityOverrideInSimSpace=true
    bOverrideAngularDamping=true
    gravityScale=0.8
    simulationSpace=BoneRelative
    alphaBoolBlend {blendOption=Linear}
    boundBone {boneName=dyn_sleeve_r_02}
    componentAppliedLinearAccClamp=<500 500 500>
    componentLinearAccScale=<1 1 1>
    componentLinearVelScale=<1 1 1>
    gravityOverride {z=-980}
    physicsBodyDefinitions [{
      sphereCollisionRadius=0
      boundBone {boneName=dyn_sleeve_r_02}
      boxExtents=<17 17 13>
      constraintSetup {angularLimitsMax=<5 15 5>, angularLimitsMin=<-5 -15 -5>, angularTarget=<1 0 0>}
      localJointOffset=<1 0 0>
     }]
    relativeSpaceBone {boneName=root}
   }
   n42 = Anim Dynamics - Bone: dyn_sleeve_r_03(ComponentPose=@n41, Alpha=1, ExternalForce="0,0,0") {
    angularDampingOverride=0.9999
    angularSpringConstant=500
    bAngularSpring=true
    bGravityOverrideInSimSpace=true
    bOverrideAngularDamping=true
    gravityScale=0.8
    simulationSpace=BoneRelative
    alphaBoolBlend {blendOption=Linear}
    boundBone {boneName=dyn_sleeve_r_03}
    componentAppliedLinearAccClamp=<500 500 500>
    componentLinearAccScale=<1 1 1>
    componentLinearVelScale=<1 1 1>
    gravityOverride {z=-980}
    physicsBodyDefinitions [{
      sphereCollisionRadius=0
      boundBone {boneName=dyn_sleeve_r_03}
      boxExtents=<17 17 13>
      constraintSetup {angularLimitsMax=<5 15 5>, angularLimitsMin=<-5 -15 -5>, angularTarget=<1 0 0>}
      localJointOffset=<1 0 0>
     }]
    relativeSpaceBone {boneName=root}
   }
   n43 = Anim Dynamics - Bone: dyn_sleeve_r_04(ComponentPose=@n42, Alpha=0.5, ExternalForce="0,0,0") {
    angularDampingOverride=0.9999
    angularSpringConstant=3500
    bAngularSpring=true
    bGravityOverrideInSimSpace=true
    bOverrideAngularDamping=true
    gravityScale=0.8
    simulationSpace=BoneRelative
    alphaBoolBlend {blendOption=Linear}
    boundBone {boneName=dyn_sleeve_r_04}
    componentAppliedLinearAccClamp=<500 500 500>
    componentLinearAccScale=<1 1 1>
    componentLinearVelScale=<1 1 1>
    gravityOverride {z=-980}
    physicsBodyDefinitions [{
      sphereCollisionRadius=0
      boundBone {boneName=dyn_sleeve_r_04}
      boxExtents=<17 17 13>
      constraintSetup {angularLimitsMax=<1 10 15>, angularLimitsMin=<-1 2 2>, angularTarget=<1 0 0>, linearAxesMax=<1.5 3 1>, linearAxesMin=<-1 0 0>}
      localJointOffset=<1 0 0>
     }]
    relativeSpaceBone {boneName=root}
   }
   n44 = Anim Dynamics - Bone: dyn_sleeve_r_04_a(ComponentPose=@n43, Alpha=0.5, ExternalForce="0,0,0") {
    angularDampingOverride=0.9999
    angularSpringConstant=500
    bAngularSpring=true
    bGravityOverrideInSimSpace=true
    bOverrideAngularDamping=true
    gravityScale=0.8
    simulationSpace=BoneRelative
    alphaBoolBlend {blendOption=Linear}
    boundBone {boneName=dyn_sleeve_r_04_a}
    componentAppliedLinearAccClamp=<500 500 500>
    componentLinearAccScale=<1 1 1>
    componentLinearVelScale=<1 1 1>
    gravityOverride {z=-980}
    physicsBodyDefinitions [{
      sphereCollisionRadius=0
      boundBone {boneName=dyn_sleeve_r_04_a}
      boxExtents=<17 17 13>
      constraintSetup {angularLimitsMax=<1 5 5>, angularLimitsMin=<-1 0 0>, angularTarget=<1 0 0>}
      localJointOffset=<1 0 0>
     }]
    relativeSpaceBone {boneName=root}
   }
   n45 = Anim Dynamics - Bone: dyn_sleeve_r_04_b(ComponentPose=@n44, Alpha=1, ExternalForce="0,0,0") {
    angularDampingOverride=0.9999
    angularSpringConstant=500
    bAngularSpring=true
    bGravityOverrideInSimSpace=true
    bOverrideAngularDamping=true
    gravityScale=0.8
    simulationSpace=BoneRelative
    alphaBoolBlend {blendOption=Linear}
    boundBone {boneName=dyn_sleeve_r_04_b}
    componentAppliedLinearAccClamp=<500 500 500>
    componentLinearAccScale=<1 1 1>
    componentLinearVelScale=<1 1 1>
    gravityOverride {z=-980}
    physicsBodyDefinitions [{
      sphereCollisionRadius=0
      boundBone {boneName=dyn_sleeve_r_04_b}
      boxExtents=<17 17 13>
      constraintSetup {angularLimitsMax=<5 15 5>, angularLimitsMin=<-5 -15 -5>, angularTarget=<1 0 0>}
      localJointOffset=<2 0 0>
     }]
    relativeSpaceBone {boneName=root}
   }
   n46 = Anim Dynamics - Bone: dyn_sleeve_r_04_c(ComponentPose=@n45, Alpha=1, ExternalForce="0,0,0") {
    angularDampingOverride=0.9999
    angularSpringConstant=500
    bAngularSpring=true
    bGravityOverrideInSimSpace=true
    bOverrideAngularDamping=true
    gravityScale=0.8
    simulationSpace=BoneRelative
    alphaBoolBlend {blendOption=Linear}
    boundBone {boneName=dyn_sleeve_r_04_c}
    componentAppliedLinearAccClamp=<500 500 500>
    componentLinearAccScale=<1 1 1>
    componentLinearVelScale=<1 1 1>
    gravityOverride {z=-980}
    physicsBodyDefinitions [{
      sphereCollisionRadius=0
      boundBone {boneName=dyn_sleeve_r_04_c}
      boxExtents=<17 17 13>
      constraintSetup {angularLimitsMax=<5 15 5>, angularLimitsMin=<-5 -15 -5>, angularTarget=<1 0 0>}
      localJointOffset=<2 0 0>
     }]
    relativeSpaceBone {boneName=root}
   }
   n47 = Anim Dynamics - Bone: dyn_knot_l_f_01(ComponentPose=@n46, Alpha=1, ExternalForce="0,0,0") {
    angularDampingOverride=0.99
    angularSpringConstant=300
    bAngularSpring=true
    bGravityOverrideInSimSpace=true
    bOverrideAngularDamping=true
    bUsePlanarLimit=false
    simulationSpace=RootRelative
    alphaBoolBlend {blendOption=Linear}
    boundBone {boneName=dyn_knot_l_f_01}
    componentAppliedLinearAccClamp=<3500 3500 3500>
    componentLinearAccScale=<0.6 0.6 0.6>
    componentLinearVelScale=<0.6 0.6 0.6>
    physicsBodyDefinitions [{
      sphereCollisionRadius=4
      boundBone {boneName=dyn_knot_l_f_01}
      boxExtents=<25 17 17>
      constraintSetup {angularLimitsMax=<1 25 0>, angularLimitsMin=<-1 -25 -35>, angularTarget=<1 0 0>}
      localJointOffset=<2 0 0>
     }]
    relativeSpaceBone {boneName=pelvis}
    sphericalLimits [{limitRadius=6, drivingBone {boneName=upperarm_twist_02_r}, sphereLocalOffset=<0 0 3>}]
   }
   n48 = Anim Dynamics - Bone: dyn_knot_l_f_02(ComponentPose=@n47, Alpha=1, ExternalForce="0,0,0") {
    angularDampingOverride=0.99
    angularSpringConstant=300
    bAngularSpring=true
    bGravityOverrideInSimSpace=true
    bOverrideAngularDamping=true
    bUsePlanarLimit=false
    simulationSpace=RootRelative
    alphaBoolBlend {blendOption=Linear}
    boundBone {boneName=dyn_knot_l_f_02}
    componentAppliedLinearAccClamp=<3500 3500 3500>
    componentLinearAccScale=<0.6 0.6 0.6>
    componentLinearVelScale=<0.6 0.6 0.6>
    physicsBodyDefinitions [{
      sphereCollisionRadius=4
      boundBone {boneName=dyn_knot_l_f_02}
      boxExtents=<25 17 17>
      constraintSetup {angularLimitsMax=<1 25 0>, angularLimitsMin=<-1 -25 -35>, angularTarget=<1 0 0>}
      localJointOffset=<2 0 0>
     }]
    relativeSpaceBone {boneName=pelvis}
    sphericalLimits [{limitRadius=6, drivingBone {boneName=upperarm_twist_02_r}, sphereLocalOffset=<0 0 3>}]
   }
   n49 = Anim Dynamics - Bone: dyn_knot_l_f_03(ComponentPose=@n48, Alpha=1, ExternalForce="0,0,0") {
    angularDampingOverride=0.99
    angularSpringConstant=300
    bAngularSpring=true
    bGravityOverrideInSimSpace=true
    bOverrideAngularDamping=true
    bUsePlanarLimit=false
    simulationSpace=RootRelative
    alphaBoolBlend {blendOption=Linear}
    boundBone {boneName=dyn_knot_l_f_03}
    componentAppliedLinearAccClamp=<3500 3500 3500>
    componentLinearAccScale=<0.6 0.6 0.6>
    componentLinearVelScale=<0.6 0.6 0.6>
    physicsBodyDefinitions [{
      sphereCollisionRadius=4
      boundBone {boneName=dyn_knot_l_f_03}
      boxExtents=<25 17 17>
      constraintSetup {angularLimitsMax=<1 25 0>, angularLimitsMin=<-1 -25 -35>, angularTarget=<1 0 0>}
      localJointOffset=<2 0 0>
     }]
    relativeSpaceBone {boneName=pelvis}
    sphericalLimits [{limitRadius=6, drivingBone {boneName=upperarm_twist_02_r}, sphereLocalOffset=<0 0 3>}]
   }
   n50 = Anim Dynamics - Bone: dyn_knot_l_b_01(ComponentPose=@n49, Alpha=1, ExternalForce="0,0,0") {
    angularDampingOverride=0.99
    angularSpringConstant=300
    bAngularSpring=true
    bGravityOverrideInSimSpace=true
    bOverrideAngularDamping=true
    bUsePlanarLimit=false
    simulationSpace=RootRelative
    alphaBoolBlend {blendOption=Linear}
    boundBone {boneName=dyn_knot_l_b_01}
    componentAppliedLinearAccClamp=<3500 3500 3500>
    componentLinearAccScale=<0.6 0.6 0.6>
    componentLinearVelScale=<0.6 0.6 0.6>
    physicsBodyDefinitions [{
      sphereCollisionRadius=4
      boundBone {boneName=dyn_knot_l_b_01}
      boxExtents=<25 17 17>
      constraintSetup {angularLimitsMax=<1 25 0>, angularLimitsMin=<-1 -25 -35>, angularTarget=<1 0 0>}
      localJointOffset=<2 0 0>
     }]
    relativeSpaceBone {boneName=pelvis}
    sphericalLimits [{limitRadius=6, drivingBone {boneName=upperarm_twist_02_r}, sphereLocalOffset=<0 0 3>}]
   }
   n51 = Anim Dynamics - Bone: dyn_knot_l_b_02(ComponentPose=@n50, Alpha=1, ExternalForce="0,0,0") {
    angularDampingOverride=0.99
    angularSpringConstant=300
    bAngularSpring=true
    bGravityOverrideInSimSpace=true
    bOverrideAngularDamping=true
    bUsePlanarLimit=false
    simulationSpace=RootRelative
    alphaBoolBlend {blendOption=Linear}
    boundBone {boneName=dyn_knot_l_b_02}
    componentAppliedLinearAccClamp=<3500 3500 3500>
    componentLinearAccScale=<0.6 0.6 0.6>
    componentLinearVelScale=<0.6 0.6 0.6>
    physicsBodyDefinitions [{
      sphereCollisionRadius=4
      boundBone {boneName=dyn_knot_l_b_02}
      boxExtents=<25 17 17>
      constraintSetup {angularLimitsMax=<1 25 0>, angularLimitsMin=<-1 -25 -35>, angularTarget=<1 0 0>}
      localJointOffset=<2 0 0>
     }]
    relativeSpaceBone {boneName=pelvis}
    sphericalLimits [{limitRadius=6, drivingBone {boneName=upperarm_twist_02_r}, sphereLocalOffset=<0 0 3>}]
   }
   n52 = Anim Dynamics - Bone: dyn_knot_l_b_03(ComponentPose=@n51, Alpha=1, ExternalForce="0,0,0") {
    angularDampingOverride=0.99
    angularSpringConstant=300
    bAngularSpring=true
    bGravityOverrideInSimSpace=true
    bOverrideAngularDamping=true
    bUsePlanarLimit=false
    simulationSpace=RootRelative
    alphaBoolBlend {blendOption=Linear}
    boundBone {boneName=dyn_knot_l_b_03}
    componentAppliedLinearAccClamp=<3500 3500 3500>
    componentLinearAccScale=<0.6 0.6 0.6>
    componentLinearVelScale=<0.6 0.6 0.6>
    physicsBodyDefinitions [{
      sphereCollisionRadius=4
      boundBone {boneName=dyn_knot_l_b_03}
      boxExtents=<25 17 17>
      constraintSetup {angularLimitsMax=<1 25 0>, angularLimitsMin=<-1 -25 -35>, angularTarget=<1 0 0>}
      localJointOffset=<2 0 0>
     }]
    relativeSpaceBone {boneName=pelvis}
    sphericalLimits [{limitRadius=6, drivingBone {boneName=upperarm_twist_02_r}, sphereLocalOffset=<0 0 3>}]
   }
   n53 = Anim Dynamics - Bone: dyn_belt_r_01(ComponentPose=@n52, Alpha=1, ExternalForce="0,0,0") {
    angularDampingOverride=0.98
    angularSpringConstant=300
    bAngularSpring=true
    bGravityOverrideInSimSpace=true
    bOverrideAngularDamping=true
    bUsePlanarLimit=false
    simulationSpace=RootRelative
    alphaBoolBlend {blendOption=Linear}
    boundBone {boneName=dyn_belt_r_01}
    componentAppliedLinearAccClamp=<3500 3500 3500>
    componentLinearAccScale=<0.8 0.8 0.8>
    componentLinearVelScale=<0.8 0.8 0.8>
    physicsBodyDefinitions [{
      sphereCollisionRadius=4
      boundBone {boneName=dyn_belt_r_01}
      boxExtents=<25 15 15>
      constraintSetup {angularLimitsMax=<15 10 45>, angularLimitsMin=<-15 -5 -45>, angularTarget=<1 0 0>}
      localJointOffset=<2 0 0>
     }]
    relativeSpaceBone {boneName=pelvis}
    sphericalLimits [{limitRadius=6, drivingBone {boneName=upperarm_twist_02_r}, sphereLocalOffset=<0 0 3>}]
   }
   n54 = Anim Dynamics - Bone: dyn_belt_r_02(ComponentPose=@n53, Alpha=1, ExternalForce="0,0,0") {
    angularDampingOverride=0.98
    angularSpringConstant=300
    bAngularSpring=true
    bGravityOverrideInSimSpace=true
    bOverrideAngularDamping=true
    bUsePlanarLimit=false
    simulationSpace=RootRelative
    alphaBoolBlend {blendOption=Linear}
    boundBone {boneName=dyn_belt_r_02}
    componentAppliedLinearAccClamp=<3500 3500 3500>
    componentLinearAccScale=<0.8 0.8 0.8>
    componentLinearVelScale=<0.8 0.8 0.8>
    physicsBodyDefinitions [{
      sphereCollisionRadius=4
      boundBone {boneName=dyn_belt_r_02}
      boxExtents=<25 15 15>
      constraintSetup {angularLimitsMax=<15 10 45>, angularLimitsMin=<-15 -5 -45>, angularTarget=<1 0 0>}
      localJointOffset=<2 0 0>
     }]
    relativeSpaceBone {boneName=pelvis}
    sphericalLimits [{limitRadius=6, drivingBone {boneName=upperarm_twist_02_r}, sphereLocalOffset=<0 0 3>}]
   }
   n55 = Anim Dynamics - Bone: dyn_belt_r_03(ComponentPose=@n54, Alpha=1, ExternalForce="0,0,0") {
    angularDampingOverride=0.98
    angularSpringConstant=300
    bAngularSpring=true
    bGravityOverrideInSimSpace=true
    bOverrideAngularDamping=true
    bUsePlanarLimit=false
    simulationSpace=RootRelative
    alphaBoolBlend {blendOption=Linear}
    boundBone {boneName=dyn_belt_r_03}
    componentAppliedLinearAccClamp=<3500 3500 3500>
    componentLinearAccScale=<0.8 0.8 0.8>
    componentLinearVelScale=<0.8 0.8 0.8>
    physicsBodyDefinitions [{
      sphereCollisionRadius=4
      boundBone {boneName=dyn_belt_r_03}
      boxExtents=<25 15 15>
      constraintSetup {angularLimitsMax=<15 10 45>, angularLimitsMin=<-15 -5 -45>, angularTarget=<1 0 0>}
      localJointOffset=<2 0 0>
     }]
    relativeSpaceBone {boneName=pelvis}
    sphericalLimits [{limitRadius=6, drivingBone {boneName=upperarm_twist_02_r}, sphereLocalOffset=<0 0 3>}]
   }
   n56 = Anim Dynamics - Bone: dyn_ankle_fang_l_01(ComponentPose=@n55, Alpha=1, ExternalForce="0,0,0") {
    simulationSpace=RootRelative
    alphaBoolBlend {blendOption=Linear}
    boundBone {boneName=dyn_ankle_fang_l_01}
    componentAppliedLinearAccClamp=<3000 3000 3000>
    componentLinearAccScale=<0.2 0.2 0.2>
    componentLinearVelScale=<0.2 0.2 0.2>
    physicsBodyDefinitions [{
      boundBone {boneName=dyn_ankle_fang_l_01}
      boxExtents=<20 15 15>
      constraintSetup {angularLimitsMax=<1 15 15>, angularLimitsMin=<-1 -15 -15>, angularTarget=<0 0 0>}
      localJointOffset=<2 0 0>
     }]
    relativeSpaceBone {boneName=foot_l}
   }
   n57 = Anim Dynamics - Bone: dyn_ankle_fang_l_02(ComponentPose=@n56, Alpha=1, ExternalForce="0,0,0") {
    simulationSpace=RootRelative
    alphaBoolBlend {blendOption=Linear}
    boundBone {boneName=dyn_ankle_fang_l_02}
    componentAppliedLinearAccClamp=<3000 3000 3000>
    componentLinearAccScale=<0.2 0.2 0.2>
    componentLinearVelScale=<0.2 0.2 0.2>
    physicsBodyDefinitions [{
      boundBone {boneName=dyn_ankle_fang_l_02}
      boxExtents=<20 15 15>
      constraintSetup {angularLimitsMax=<1 15 15>, angularLimitsMin=<-1 -15 -15>, angularTarget=<0 0 0>}
      localJointOffset=<2 0 0>
     }]
    relativeSpaceBone {boneName=foot_l}
   }
   n58 = Anim Dynamics - Bone: dyn_ankle_fang_l_03(ComponentPose=@n57, Alpha=1, ExternalForce="0,0,0") {
    simulationSpace=RootRelative
    alphaBoolBlend {blendOption=Linear}
    boundBone {boneName=dyn_ankle_fang_l_03}
    componentAppliedLinearAccClamp=<3000 3000 3000>
    componentLinearAccScale=<0.2 0.2 0.2>
    componentLinearVelScale=<0.2 0.2 0.2>
    physicsBodyDefinitions [{
      boundBone {boneName=dyn_ankle_fang_l_03}
      boxExtents=<20 15 15>
      constraintSetup {angularLimitsMax=<1 15 15>, angularLimitsMin=<-1 -15 -15>, angularTarget=<0 0 0>}
      localJointOffset=<2 0 0>
     }]
    relativeSpaceBone {boneName=foot_l}
   }
   n59 = Anim Dynamics - Bone: dyn_ankle_fang_l_04(ComponentPose=@n58, Alpha=1, ExternalForce="0,0,0") {
    simulationSpace=RootRelative
    alphaBoolBlend {blendOption=Linear}
    boundBone {boneName=dyn_ankle_fang_l_04}
    componentAppliedLinearAccClamp=<3000 3000 3000>
    componentLinearAccScale=<0.2 0.2 0.2>
    componentLinearVelScale=<0.2 0.2 0.2>
    physicsBodyDefinitions [{
      boundBone {boneName=dyn_ankle_fang_l_04}
      boxExtents=<20 15 15>
      constraintSetup {angularLimitsMax=<1 15 15>, angularLimitsMin=<-1 -15 -15>, angularTarget=<0 0 0>}
      localJointOffset=<2 0 0>
     }]
    relativeSpaceBone {boneName=foot_l}
   }
   n60 = Anim Dynamics - Bone: dyn_ankle_fang_l_05(ComponentPose=@n59, Alpha=1, ExternalForce="0,0,0") {
    simulationSpace=RootRelative
    alphaBoolBlend {blendOption=Linear}
    boundBone {boneName=dyn_ankle_fang_l_05}
    componentAppliedLinearAccClamp=<3000 3000 3000>
    componentLinearAccScale=<0.2 0.2 0.2>
    componentLinearVelScale=<0.2 0.2 0.2>
    physicsBodyDefinitions [{
      boundBone {boneName=dyn_ankle_fang_l_05}
      boxExtents=<20 15 15>
      constraintSetup {angularLimitsMax=<1 15 15>, angularLimitsMin=<-1 -15 -15>, angularTarget=<0 0 0>}
      localJointOffset=<2 0 0>
     }]
    relativeSpaceBone {boneName=foot_l}
   }
   n61 = Anim Dynamics - Bone: dyn_ankle_fang_l_06(ComponentPose=@n60, Alpha=1, ExternalForce="0,0,0") {
    simulationSpace=RootRelative
    alphaBoolBlend {blendOption=Linear}
    boundBone {boneName=dyn_ankle_fang_l_06}
    componentAppliedLinearAccClamp=<3000 3000 3000>
    componentLinearAccScale=<0.2 0.2 0.2>
    componentLinearVelScale=<0.2 0.2 0.2>
    physicsBodyDefinitions [{
      boundBone {boneName=dyn_ankle_fang_l_06}
      boxExtents=<20 15 15>
      constraintSetup {angularLimitsMax=<1 15 15>, angularLimitsMin=<-1 -15 -15>, angularTarget=<0 0 0>}
      localJointOffset=<2 0 0>
     }]
    relativeSpaceBone {boneName=foot_l}
   }
   n62 = Anim Dynamics - Bone: dyn_ankle_fang_l_07(ComponentPose=@n61, Alpha=1, ExternalForce="0,0,0") {
    simulationSpace=RootRelative
    alphaBoolBlend {blendOption=Linear}
    boundBone {boneName=dyn_ankle_fang_l_07}
    componentAppliedLinearAccClamp=<3000 3000 3000>
    componentLinearAccScale=<0.2 0.2 0.2>
    componentLinearVelScale=<0.2 0.2 0.2>
    physicsBodyDefinitions [{
      boundBone {boneName=dyn_ankle_fang_l_07}
      boxExtents=<20 15 15>
      constraintSetup {angularLimitsMax=<1 15 15>, angularLimitsMin=<-1 -15 -15>, angularTarget=<0 0 0>}
      localJointOffset=<2 0 0>
     }]
    relativeSpaceBone {boneName=foot_l}
   }
   n63 = Anim Dynamics - Bone: dyn_ankle_fang_l_08(ComponentPose=@n62, Alpha=1, ExternalForce="0,0,0") {
    simulationSpace=RootRelative
    alphaBoolBlend {blendOption=Linear}
    boundBone {boneName=dyn_ankle_fang_l_08}
    componentAppliedLinearAccClamp=<3000 3000 3000>
    componentLinearAccScale=<0.2 0.2 0.2>
    componentLinearVelScale=<0.2 0.2 0.2>
    physicsBodyDefinitions [{
      boundBone {boneName=dyn_ankle_fang_l_08}
      boxExtents=<20 15 15>
      constraintSetup {angularLimitsMax=<1 15 15>, angularLimitsMin=<-1 -15 -15>, angularTarget=<0 0 0>}
      localJointOffset=<2 0 0>
     }]
    relativeSpaceBone {boneName=foot_l}
   }
   n64 = Anim Dynamics - Bone: dyn_ankle_fang_r_01(ComponentPose=@n63, Alpha=1, ExternalForce="0,0,0") {
    simulationSpace=RootRelative
    alphaBoolBlend {blendOption=Linear}
    boundBone {boneName=dyn_ankle_fang_r_01}
    componentAppliedLinearAccClamp=<3000 3000 3000>
    componentLinearAccScale=<0.2 0.2 0.2>
    componentLinearVelScale=<0.2 0.2 0.2>
    physicsBodyDefinitions [{
      boundBone {boneName=dyn_ankle_fang_r_01}
      boxExtents=<20 15 15>
      constraintSetup {angularLimitsMax=<1 15 15>, angularLimitsMin=<-1 -15 -15>, angularTarget=<0 0 0>}
      localJointOffset=<2 0 0>
     }]
    relativeSpaceBone {boneName=foot_r}
   }
   n65 = Anim Dynamics - Bone: dyn_ankle_fang_r_02(ComponentPose=@n64, Alpha=1, ExternalForce="0,0,0") {
    simulationSpace=RootRelative
    alphaBoolBlend {blendOption=Linear}
    boundBone {boneName=dyn_ankle_fang_r_02}
    componentAppliedLinearAccClamp=<3000 3000 3000>
    componentLinearAccScale=<0.2 0.2 0.2>
    componentLinearVelScale=<0.2 0.2 0.2>
    physicsBodyDefinitions [{
      boundBone {boneName=dyn_ankle_fang_r_02}
      boxExtents=<20 15 15>
      constraintSetup {angularLimitsMax=<1 15 15>, angularLimitsMin=<-1 -15 -15>, angularTarget=<0 0 0>}
      localJointOffset=<2 0 0>
     }]
    relativeSpaceBone {boneName=foot_r}
   }
   n66 = Anim Dynamics - Bone: dyn_ankle_fang_r_03(ComponentPose=@n65, Alpha=1, ExternalForce="0,0,0") {
    simulationSpace=RootRelative
    alphaBoolBlend {blendOption=Linear}
    boundBone {boneName=dyn_ankle_fang_r_03}
    componentAppliedLinearAccClamp=<3000 3000 3000>
    componentLinearAccScale=<0.2 0.2 0.2>
    componentLinearVelScale=<0.2 0.2 0.2>
    physicsBodyDefinitions [{
      boundBone {boneName=dyn_ankle_fang_r_03}
      boxExtents=<20 15 15>
      constraintSetup {angularLimitsMax=<1 15 15>, angularLimitsMin=<-1 -15 -15>, angularTarget=<0 0 0>}
      localJointOffset=<2 0 0>
     }]
    relativeSpaceBone {boneName=foot_r}
   }
   n67 = Anim Dynamics - Bone: dyn_ankle_fang_r_04(ComponentPose=@n66, Alpha=1, ExternalForce="0,0,0") {
    simulationSpace=RootRelative
    alphaBoolBlend {blendOption=Linear}
    boundBone {boneName=dyn_ankle_fang_r_04}
    componentAppliedLinearAccClamp=<3000 3000 3000>
    componentLinearAccScale=<0.2 0.2 0.2>
    componentLinearVelScale=<0.2 0.2 0.2>
    physicsBodyDefinitions [{
      boundBone {boneName=dyn_ankle_fang_r_04}
      boxExtents=<20 15 15>
      constraintSetup {angularLimitsMax=<1 15 15>, angularLimitsMin=<-1 -15 -15>, angularTarget=<0 0 0>}
      localJointOffset=<2 0 0>
     }]
    relativeSpaceBone {boneName=foot_r}
   }
   n68 = Anim Dynamics - Bone: dyn_ankle_fang_r_05(ComponentPose=@n67, Alpha=1, ExternalForce="0,0,0") {
    simulationSpace=RootRelative
    alphaBoolBlend {blendOption=Linear}
    boundBone {boneName=dyn_ankle_fang_r_05}
    componentAppliedLinearAccClamp=<3000 3000 3000>
    componentLinearAccScale=<0.2 0.2 0.2>
    componentLinearVelScale=<0.2 0.2 0.2>
    physicsBodyDefinitions [{
      boundBone {boneName=dyn_ankle_fang_r_05}
      boxExtents=<20 15 15>
      constraintSetup {angularLimitsMax=<1 15 15>, angularLimitsMin=<-1 -15 -15>, angularTarget=<0 0 0>}
      localJointOffset=<2 0 0>
     }]
    relativeSpaceBone {boneName=foot_r}
   }
   n69 = Anim Dynamics - Bone: dyn_ankle_fang_r_06(ComponentPose=@n68, Alpha=1, ExternalForce="0,0,0") {
    simulationSpace=RootRelative
    alphaBoolBlend {blendOption=Linear}
    boundBone {boneName=dyn_ankle_fang_r_06}
    componentAppliedLinearAccClamp=<3000 3000 3000>
    componentLinearAccScale=<0.2 0.2 0.2>
    componentLinearVelScale=<0.2 0.2 0.2>
    physicsBodyDefinitions [{
      boundBone {boneName=dyn_ankle_fang_r_06}
      boxExtents=<20 15 15>
      constraintSetup {angularLimitsMax=<1 15 15>, angularLimitsMin=<-1 -15 -15>, angularTarget=<0 0 0>}
      localJointOffset=<2 0 0>
     }]
    relativeSpaceBone {boneName=foot_r}
   }
   n70 = Anim Dynamics - Bone: dyn_ankle_fang_r_07(ComponentPose=@n69, Alpha=1, ExternalForce="0,0,0") {
    simulationSpace=RootRelative
    alphaBoolBlend {blendOption=Linear}
    boundBone {boneName=dyn_ankle_fang_r_07}
    componentAppliedLinearAccClamp=<3000 3000 3000>
    componentLinearAccScale=<0.2 0.2 0.2>
    componentLinearVelScale=<0.2 0.2 0.2>
    physicsBodyDefinitions [{
      boundBone {boneName=dyn_ankle_fang_r_07}
      boxExtents=<20 15 15>
      constraintSetup {angularLimitsMax=<1 15 15>, angularLimitsMin=<-1 -15 -15>, angularTarget=<0 0 0>}
      localJointOffset=<2 0 0>
     }]
    relativeSpaceBone {boneName=foot_r}
   }
   n71 = Anim Dynamics - Bone: dyn_ankle_fang_r_08(ComponentPose=@n70, Alpha=1, ExternalForce="0,0,0") {
    simulationSpace=RootRelative
    alphaBoolBlend {blendOption=Linear}
    boundBone {boneName=dyn_ankle_fang_r_08}
    componentAppliedLinearAccClamp=<3000 3000 3000>
    componentLinearAccScale=<0.2 0.2 0.2>
    componentLinearVelScale=<0.2 0.2 0.2>
    physicsBodyDefinitions [{
      boundBone {boneName=dyn_ankle_fang_r_08}
      boxExtents=<20 15 15>
      constraintSetup {angularLimitsMax=<1 15 15>, angularLimitsMin=<-1 -15 -15>, angularTarget=<0 0 0>}
      localJointOffset=<2 0 0>
     }]
    relativeSpaceBone {boneName=foot_r}
   }
   n72 = Get jumping()
   n73 = "Transform (Modify) Bone - Bone: dyn_sleeve_l_04"(ComponentPose=@n71, Alpha=@n72, Translation="0,3,0", Rotation="0,0,0", Scale="1,1,1") {
    rotationMode=BMM_Additive
    rotationSpace=BCS_BoneSpace
    translationMode=BMM_Additive
    translationSpace=BCS_BoneSpace
    alphaBoolBlend {blendOption=Linear}
    alphaScaleBiasClamp {bInterpResult=true, interpSpeedIncreasing=2}
    boneToModify {boneName=dyn_sleeve_l_04}
    translation {y=3}
   }
   n74 = "Transform (Modify) Bone - Bone: dyn_sleeve_r_04"(ComponentPose=@n73, Alpha=@n72, Translation="0,3,0", Rotation="0,0,0", Scale="1,1,1") {
    rotationMode=BMM_Additive
    rotationSpace=BCS_BoneSpace
    translationMode=BMM_Additive
    translationSpace=BCS_BoneSpace
    alphaBoolBlend {blendOption=Linear}
    alphaScaleBiasClamp {bInterpResult=true, interpSpeedIncreasing=2}
    boneToModify {boneName=dyn_sleeve_r_04}
    translation {y=3}
   }
   n75 = "Transform (Modify) Bone - Bone: dyn_sleeve_l_01"(ComponentPose=@n74, Alpha=@n72, Translation="0,-2,0", Rotation="0,0,0", Scale="1,1,1") {
    rotationMode=BMM_Additive
    rotationSpace=BCS_BoneSpace
    translationMode=BMM_Additive
    translationSpace=BCS_BoneSpace
    alphaBoolBlend {blendOption=Linear}
    alphaScaleBiasClamp {bInterpResult=true, interpSpeedIncreasing=2}
    boneToModify {boneName=dyn_sleeve_l_01}
    translation {y=-2}
   }
   n76 = "Transform (Modify) Bone - Bone: dyn_sleeve_r_01"(ComponentPose=@n75, Alpha=@n72, Translation="0,-2,0", Rotation="0,0,0", Scale="1,1,1") {
    rotationMode=BMM_Additive
    rotationSpace=BCS_BoneSpace
    translationMode=BMM_Additive
    translationSpace=BCS_BoneSpace
    alphaBoolBlend {blendOption=Linear}
    alphaScaleBiasClamp {bInterpResult=true, interpSpeedIncreasing=2}
    boneToModify {boneName=dyn_sleeve_r_01}
    translation {y=-2}
   }
   n77 = "Transform (Modify) Bone - Bone: dyn_l_mouth_upper"(ComponentPose=@n76, Alpha=1, Translation=Get mouth movement 1() -> (mouth movement 1:Vector), Rotation="0,0,0", Scale="1,1,1") {
    rotationSpace=BCS_WorldSpace
    translationMode=BMM_Additive
    translationSpace=BCS_BoneSpace
    alphaBoolBlend {blendOption=Linear}
    alphaScaleBiasClamp {interpSpeedDecreasing=2, interpSpeedIncreasing=2}
    boneToModify {boneName=dyn_l_mouth_upper}
   }
   n78 = "Transform (Modify) Bone - Bone: dyn_r_mouth_upper"(ComponentPose=@n77, Alpha=1, Translation=Get mouth movement 2() -> (mouth movement 2:Vector), Rotation="0,0,0", Scale="1,1,1") {
    rotationSpace=BCS_WorldSpace
    translationMode=BMM_Additive
    translationSpace=BCS_BoneSpace
    alphaBoolBlend {blendOption=Linear}
    alphaScaleBiasClamp {interpSpeedDecreasing=2, interpSpeedIncreasing=2}
    boneToModify {boneName=dyn_r_mouth_upper}
   }
   n79 = Control Rig!(
    Source=Component To Local(ComponentPose=@n78)
    Alpha
    mouth bool:bool=Get mouth bool check() -> (mouth bool check:bool)
    mouth float:double=Get mouth smile() -> (mouth smile:double)
    mouth bool_0:bool=Get mouth bool check_0() -> (mouth bool check_0:bool)
    mouth float_0:double=Get mouth smile_0() -> (mouth smile_0:double)
   ) -> (Pose) {
    alphaBoolBlend {blendOption=Linear}
    controlRigAssetReference {blueprintRigClass=&/F_MED_MelodyUrchin_tongues_CtrlRig.F_MED_MelodyUrchin_tongues_CtrlRig_C}
   }
   Output Pose(Result=@n79) {layerGroup=None}
   Get jumping()
  ]}
 EventGraph {nodes [
   Event BlueprintUpdateAnimation()
   > n1 = Sequence!(execute) -> (then_0, then_1, then_2:Exec, then_3:Exec)
   > n2 = Gravity Override State Function DEPRECATED(CharacterPartAnimInstance=Self-Reference(), GameTimeInSeconds=GetGameTimeInSeconds(), GravityOverrideParams=Get Gravity Override Params())
   > Set Gravity Override(Gravity Override=@n2)
   > n3 = Branch(Condition=OR Boolean(A=Get bIsJumping(), B=Get bIsFalling()))
   > Set jumping(jumping:double=1) -> (Output_Get:double)
   > n4 = Branch(Condition=Get bIsSkyDiving())
   Set jumping(execute=[@n4.then, @n3.else], jumping=0)
   n5 = PlayerMovement(execute=@n1.then_1, AnimInstance=Self-Reference(), XDotProductCoefficient=1, YDotProductCoefficient=1, ZDotProductCoefficient=1)
   > Set RightMovement(RightMovement:float=MapRangeClamped(Value=@n5, InRangeA=0.2, InRangeB=1)) -> (Output_Get:float)
   > Set LeftMovement(LeftMovement:float=MapRangeClamped(Value=@n5, InRangeA=-0.2, InRangeB=-1)) -> (Output_Get:float)
   > Set bwdMovement(bwdMovement:float=MapRangeClamped(Value=@n5.YMovement, InRangeA=-0.4, InRangeB=-1)) -> (Output_Get:float)
   > Set emote(emote:double="To Float (Boolean)"(InBool=Get bIsPlayingEmote())) -> (Output_Get:double)
   n6 = Branch(execute=@n1.then_2, Condition=Get bIsCrouching())
   n7 = Get LeftMovement()
   n8 = Get bwdMovement() -> (bwdMovement:float)
   > Set crouch strafe bk l(crouch strafe bk l:float=MapRangeClamped(Value=float * float(A=@n7, B=@n8) {is=title K2Node_PromotableOperator}, InRangeA=0, InRangeB=0.3)) -> (Output_Get:float)
   n9 = Get bwdMovement()
   n10 = Get RightMovement()
   > Set crouch strafe bk r(crouch strafe bk r:float=MapRangeClamped(Value=float * float(A=@n10, B=@n9), InRangeA=0, InRangeB=0.3)) -> (Output_Get:float)
   > Set crouching(crouching:double=1) -> (Output_Get:double)
   Set crouch strafe bk l(execute=@n6.else, crouch strafe bk l=0)
   > Set crouch strafe bk r(crouch strafe bk r=0)
   > Set crouching(crouching=0)
   n11 = Random Vector Function!(
    execute=@n1.then_3
    CharacterPartAnimInstance:Custom Character Part Anim Instance Object Reference
    SkelMeshComponent:Mesh Component Object Reference=GetOwnerSkeletalMeshComponent()
    JointName:Name=None
    GameTimeInSeconds:double=GetGameTimeInSeconds()
    TimeOffset:double=0
    PerlinScale:double=0
    Frequency:Vector="0.7,0.8,0.75"
    Amplitude:Vector="1.5,0.4,0.3"
    Offset:Vector="1.5,0,0"
   ) -> (then, VectorOut:Vector)
   > Set mouth movement 1(mouth movement 1:Vector=@n11) -> (Output_Get:Vector)
   > n12 = Random Vector Function(
    SkelMeshComponent=GetOwnerSkeletalMeshComponent()
    JointName=None
    GameTimeInSeconds=GetGameTimeInSeconds()
    TimeOffset=0
    PerlinScale=0
    Frequency="1.1,0.8,0.75"
    Amplitude="1.3,0.5,0.4"
    Offset="1.5,0,0"
   )
   > Set mouth movement 2(mouth movement 2:Vector=@n12) -> (Output_Get:Vector)
   > Set mouth(mouth:double=GetAlphaFromReferenceTransformDelta(
     SkeletalMeshComponent=Get Skel_Mesh()
     BoneName=dyn_r_mouth_upper
     BaseName=calf_r
     DeltaComponentType=TranslationZ
     InputRange=(X=0,Y=1)
     TwistAxis="1,0,0"
    )) -> (Output_Get:double)
   > n13 = Branch(Condition=AND Boolean(A=Get mouth bool check(), B=Get mouth bool check_0()))
   Delay(execute=@n13.else, Duration=15)
   > n14 = Sequence!(execute) -> (then_0, then_1)
   > n15 = Branch(Condition=Get mouth bool check_0())
   > Set mouth bool check(mouth bool check:bool=false) -> (Output_Get:bool)
   n16 = RandomFloatInRange(Min=0, Max=1)
   Set mouth smile(execute=@n15.else, mouth smile:double=SelectFloat(A=1, B=0, bPickA="float > float"(A=@n16, B=0.8))) -> (Output_Get:double)
   > Set mouth bool check(mouth bool check=true)
   > Delay(Duration=RandomFloatInRange(Min=1, Max=3.5))
   > Set mouth smile(mouth smile=0)
   > Set mouth bool check(mouth bool check=false)
   Delay(execute=@n14.then_1, Duration=RandomFloatInRange(Min=5, Max=9))
   > n17 = Branch(Condition=Get mouth bool check())
   > Set mouth bool check_0(mouth bool check_0:bool=false) -> (Output_Get:bool)
   n18 = RandomFloatInRange(Min=0, Max=1)
   Set mouth smile_0(execute=@n17.else, mouth smile_0:double=SelectFloat(A=1, B=0, bPickA="float > float"(A=@n18, B=0.8))) -> (Output_Get:double)
   > Set mouth bool check_0(mouth bool check_0=true)
   > Delay(Duration=RandomFloatInRange(Min=1, Max=3.5))
   > Set mouth smile_0(mouth smile_0=0)
   > Set mouth bool check_0(mouth bool check_0=false)
   Event BlueprintInitializeAnimation()
   > Set Skel_Mesh(Skel_Mesh=GetOwnerSkeletalMeshComponent())
   Branch(Condition=Get mouth bool check_0())
   Branch(Condition=Get mouth bool check())
  ]}
]
variables [
 bwdMovement {type=number}
 crouch strafe bk l {type=number}
 crouch strafe bk r {type=number}
 crouching {type=number}
 emote {type=number}
 Gravity Override {type=Vector}
 Gravity Override Params {
  type=GravityOverrideParamsStruct
  default {skydive_FloatingDown {gravityOverride=<0 -400 100>, windAmplitude=<75 -125 175>, windFrequency=<1.5 4 3>, windOffset=<0 -300 200>}}
 }
 jumping {type=number}
 LeftMovement {type=number}
 mouth {type=number}
 mouth bool check {type=boolean}
 mouth bool check_0 {type=boolean}
 mouth movement 1 {type=Vector}
 mouth movement 2 {type=Vector}
 mouth smile {type=number}
 mouth smile_0 {type=number}
 RightMovement {type=number}
 Skel_Mesh {type=SkeletalMeshComponent}
 Stream {type=RandomStream, default {initialSeed=0, seed=0}}
]
